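# Sound cue for a message prompt in the Crucible limited-time mode, playing a single prompt wave.
# A fixed short Delay node feeds one Wave Player at half volume.
# The cue is assigned to the base music sound class with a very large max distance.
# A concurrency override allows one instance at a time and blocks new plays while one is sounding.
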SoundCue /Game/Athena/Sounds/LTMs/Crucible/Crucible_MessagePrompt_Cue
root=Delay_0
nodes [
 {id=Delay_0, children [WavePlayer_0], properties {delayMax=0.1, delayMin=0.1}}
 {id=WavePlayer_0, properties {soundWaveAssetPtr=&/Crucible_MessagePrompt_01}}
]
properties {
 bOverrideConcurrency=true
 maxDistance=2097152
 soundClassObject=/Game/Sounds/FortSoundClassesAndModes/Fort_11_Base_Music.Fort_11_Base_music
 volumeMultiplier=0.5
 concurrencyOverrides {maxCount=1, resolutionRule=PreventNew}
}
