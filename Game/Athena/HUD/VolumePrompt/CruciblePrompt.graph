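# HUD volume prompt widget for the Crucible LTM, built on FortHUDTagPromptWidgetBase, showing a title and description framed by filigree art.
# OpenWidget fills both texts from the prompt data, rebinds the Outro-finished delegate, plays the Intro animation and a sound cue.
# Intro animates filigree material reveal parameters and fades in elements.
# A retriggerable delay later plays the Outro fade, so repeated opens extend the display; CloseWidget collapses the widget.
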WidgetBlueprint /Game/Athena/HUD/VolumePrompt/CruciblePrompt
animations [
 Intro {
  binding_count=5
  length_seconds=2.0518
  object=&self:Intro
  sequence {bindings [
    Image_FiligreeLeft {
     id=#1
     possessed_class=Image
     tracks [{
       type=MovieSceneWidgetMaterialTrack
       sections [{
         type=MovieSceneParameterSection
         channels [
          RevealRadius {type=MovieSceneScriptingFloatChannel, keys [[0,4.1453],[13.5187,0]]}
          RevealDensity {type=MovieSceneScriptingFloatChannel, keys [[0,1.4221],[14.0137,1.489]]}
         ]
        }]
      }]
    }
    Text_Description {
     id=#2
     possessed_class=CommonTextBlock
     tracks [{
       type=MovieSceneFloatTrack
       display_name=Render Opacity
       sections [{type=MovieSceneFloatSection, channels [None {type=MovieSceneScriptingFloatChannel, keys [[15.845,0],[41.036,1]]}]}]
      }]
    }
    Image_Header {
     id=#3
     possessed_class=Image
     tracks [{
       type=MovieSceneFloatTrack
       display_name=Render Opacity
       sections [{type=MovieSceneFloatSection, end_frame=14, start_frame=0, channels [None {type=MovieSceneScriptingFloatChannel, keys [[10,0],[14.0137,1]]}]}]
      }]
    }
    Overlay_Container {
     id=#4
     possessed_class=Overlay
     tracks [{
       type=MovieSceneFloatTrack
       display_name=Render Opacity
       sections [{type=MovieSceneFloatSection, end_frame=14, start_frame=0, channels [None {type=MovieSceneScriptingFloatChannel, keys [[0,0],[14.7813,1]]}]}]
      }]
    }
    Image_FiligreeRight {
     id=#5
     possessed_class=Image
     tracks [{
       type=MovieSceneWidgetMaterialTrack
       sections [{
         type=MovieSceneParameterSection
         channels [
          RevealRadius {type=MovieSceneScriptingFloatChannel, keys [[0,4.1453],[13.5187,0]]}
          RevealDensity {type=MovieSceneScriptingFloatChannel, keys [[0,1.4221],[14.0137,1.489]]}
         ]
        }]
      }]
    }
   ]}
 }
 Outro {
  binding_count=1
  length_seconds=0.7391
  object=&self:Outro
  sequence {bindings [Overlay_Container {
     id=#6
     possessed_class=Overlay
     tracks [{
       type=MovieSceneFloatTrack
       display_name=Render Opacity
       sections [{type=MovieSceneFloatSection, end_frame=14, start_frame=0, channels [None {type=MovieSceneScriptingFloatChannel, keys [[0,1],[14.7813,0]]}]}]
      }]
    }]}
 }
]
graphs [
 EventGraph {nodes [
   Event CloseWidget() {is=title K2Node_Event}
   > Parent: CloseWidget() {is=title K2Node_CallParentFunction}
   > SetVisibility(InVisibility=Collapsed)
   n1 = Event OpenWidget!() -> (OutputDelegate, then, PrompData:"Fort HUDTag Prompt Data Structure (by ref)") {is=title K2Node_Event}
   > Parent: OpenWidget!(execute, PrompData:"Fort HUDTag Prompt Data Structure (by ref)"=@n1.PrompData) -> (then) {is=title K2Node_CallParentFunction}
   > K2Node_Composite_0 = Set Text and Images!(execute, VolumePromptData:"Fort HUDTag Prompt Data Structure (by ref)"=@n1.PrompData) -> (then:Exec) {is=title K2Node_Composite}
   > UnbindAllFromAnimationFinished!(
    execute
    self:User Widget Object Reference
    Animation:Widget Animation Object Reference=Get Outro() -> (Outro:Widget Animation Object Reference)
   ) -> (then)
   > StopAnimation(InAnimation=Get Outro())
   > BindToAnimationFinished(Animation=Get Outro(), Delegate=Create Event())
   > SetVisibility(InVisibility=HitTestInvisible)
   > PlayAnimationForward(InAnimation=Get Intro())
   > PlaySound2D(Sound=Get SFX_Intro() -> (SFX_Intro:Sound Base Object Reference))
   > RetriggerableDelay(Duration=7)
   > PlayAnimation(InAnimation=Get Outro())
  ]}
 Set Text and Images {
  path=EventGraph.K2Node_Composite_0.Set Text and Images
  nodes [
   n1 = Inputs!() -> (execute, VolumePromptData:"Fort HUDTag Prompt Data Structure (by ref)")
   n2 = Break Fort HUDTag Prompt Data!(FortHUDTagPromptData:"Fort HUDTag Prompt Data Structure (by ref)"=@n1) -> (
    VolumeTag:Gameplay Tag Structure
    Text_PromptTitle:Text
    Text_PromptDescription:Text
    bIsEnterData:bool=~
    WidgetClass:User Widget Class Reference
   ) {is=title K2Node_BreakStruct}
   > "SetText (Text)"(self=Get Text_Title() -> (Text_Title:Common Text Object Reference), InText=@n2.Text_PromptTitle)
   > "SetText (Text)"(self=Get Text_Description() -> (Text_Description:Common Text Object Reference), InText=@n2.Text_PromptDescription)
   > Outputs!(then:Exec)
  ]
 }
]
variables [SFX_Intro {type=SoundBase, default=/Game/Athena/Sounds/LTMs/Crucible/Crucible_MessagePrompt_Cue}]
widgets {
 info {inheritedWidgetCount=1, parentClass=/Script/FortniteUI.FortHUDTagPromptWidgetBase, rootWidgetClass=&Overlay, widgetCount=18}
 widgets [
  Overlay_Container {widgetClassPath=&Overlay, widget=Overlay_Container}
  Image_Shadow {
   widgetClassPath=&Image
   bIsVariable=true
   parent=Overlay_Container
   widget=Image_Shadow
   properties {
    renderOpacity=0.8
    brush {
     bIsDynamicallyLoaded=false
     drawAs=Image
     imageType=NoImage
     mirroring=NoMirror
     resourceName=None
     resourceObject=/Game/Athena/HUD/LTM/Wax/Materials/M_UI_TextGlow
     tiling=NoTile
     imageSize=<1600 400>
     margin=<ltrb 0 0 0 0>
     outlineSettings {
      bUseBrushTransparency=false
      roundingType=HalfHeightRadius
      width=0
      color {colorUseRule=UseColor_Specified, specifiedColor=<rgba 0 0 0 0>}
      cornerRadii=<0 0 0 0>
     }
     tintColor {colorUseRule=UseColor_Specified, specifiedColor=<rgba 1 1 1 1>}
     uVRegion {bIsValid=false, max=<0 0>, min=<0 0>}
    }
    colorAndOpacity=<rgba 0.4314 0.5681 0.5833 0.5>
   }
   slot {class=OverlaySlot, properties {horizontalAlignment=HAlign_Center, verticalAlignment=VAlign_Center, padding=<ltrb -800 -200 -800 -200>}}
  }
  HorizontalBox_0 {
   widgetClassPath=&HorizontalBox
   parent=Overlay_Container
   widget=HorizontalBox_0
   slot {class=OverlaySlot, properties {horizontalAlignment=HAlign_Fill, padding=<ltrb -40 0 -40 0>}}
  }
  Overlay_0 {
   widgetClassPath=&Overlay
   parent=HorizontalBox_0
   widget=Overlay_0
   slot {class=HorizontalBoxSlot, properties {size {sizeRule=Fill, value=1}}}
  }
  Image_FiligreeLeft {
   widgetClassPath=&Image
   bIsVariable=true
   parent=Overlay_0
   widget=Image_FiligreeLeft
   properties {
    brush {
     bIsDynamicallyLoaded=false
     drawAs=Box
     imageType=NoImage
     mirroring=NoMirror
     resourceName=None
     resourceObject=/Game/Athena/HUD/VolumePrompt/Materials/M_UI_CrucibleBorder
     tiling=NoTile
     imageSize=<50 120>
     margin=<ltrb 0.1 0 0.9 0>
     outlineSettings {
      bUseBrushTransparency=false
      roundingType=HalfHeightRadius
      width=0
      color {colorUseRule=UseColor_Specified, specifiedColor=<rgba 0 0 0 0>}
      cornerRadii=<0 0 0 0>
     }
     tintColor {colorUseRule=UseColor_Specified, specifiedColor=<rgba 1 1 1 1>}
     uVRegion {bIsValid=false, max=<0 0>, min=<0 0>}
    }
    renderTransform {angle=0, scale=<-1 1>, shear=<0 0>, translation=<0 0>}
   }
   slot {class=OverlaySlot, properties {horizontalAlignment=HAlign_Fill}}
  }
  Overlay_1 {
   widgetClassPath=&Overlay
   parent=HorizontalBox_0
   widget=Overlay_1
   slot {class=HorizontalBoxSlot, properties {size {sizeRule=Fill, value=1}}}
  }
  Image_FiligreeRight {
   widgetClassPath=&Image
   bIsVariable=true
   parent=Overlay_1
   widget=Image_FiligreeRight
   properties {brush {
     bIsDynamicallyLoaded=false
     drawAs=Box
     imageType=NoImage
     mirroring=NoMirror
     resourceName=None
     resourceObject=&M_UI_CrucibleBorder
     tiling=NoTile
     imageSize=<50 120>
     margin=<ltrb 0.1 0 0.9 0>
     outlineSettings {
      bUseBrushTransparency=false
      roundingType=HalfHeightRadius
      width=0
      color {colorUseRule=UseColor_Specified, specifiedColor=<rgba 0 0 0 0>}
      cornerRadii=<0 0 0 0>
     }
     tintColor {colorUseRule=UseColor_Specified, specifiedColor=<rgba 1 1 1 1>}
     uVRegion {bIsValid=false, max=<0 0>, min=<0 0>}
    }}
   slot {class=OverlaySlot, properties {horizontalAlignment=HAlign_Fill}}
  }
  Overlay_2 {
   widgetClassPath=&Overlay
   parent=Overlay_Container
   widget=Overlay_2
   slot {class=OverlaySlot, properties {horizontalAlignment=HAlign_Center}}
  }
  Image_Header {
   widgetClassPath=&Image
   bIsVariable=true
   parent=Overlay_2
   widget=Image_Header
   properties {
    brush {
     bIsDynamicallyLoaded=false
     drawAs=Box
     imageType=NoImage
     mirroring=NoMirror
     resourceName=None
     resourceObject=/Game/Athena/HUD/VolumePrompt/Materials/M_UI_CrucibleHeader
     tiling=NoTile
     imageSize=<60 50>
     margin=<ltrb 0 0 0 0>
     outlineSettings {
      bUseBrushTransparency=false
      roundingType=HalfHeightRadius
      width=0
      color {colorUseRule=UseColor_Specified, specifiedColor=<rgba 0 0 0 0>}
      cornerRadii=<0 0 0 0>
     }
     tintColor {colorUseRule=UseColor_Specified, specifiedColor=<rgba 1 1 1 1>}
     uVRegion {bIsValid=false, max=<0 0>, min=<0 0>}
    }
    renderTransform {angle=0, scale=<-1 1>, shear=<0 0>, translation=<0 -28.5>}
   }
   slot {class=OverlaySlot}
  }
  VerticalBox_0 {
   widgetClassPath=&VerticalBox
   parent=Overlay_Container
   widget=VerticalBox_0
   slot {class=OverlaySlot, properties {verticalAlignment=VAlign_Center}}
  }
  SizeBox_1 {
   widgetClassPath=&SizeBox
   parent=VerticalBox_0
   widget=SizeBox_1
   properties {bOverride_MaxDesiredWidth=true, maxDesiredWidth=650}
   slot {class=VerticalBoxSlot}
  }
  ScaleBox_1 {widgetClassPath=&ScaleBox, parent=SizeBox_1, widget=ScaleBox_1, properties {stretchDirection=DownOnly}, slot {class=SizeBoxSlot}}
  Text_Title {
   widgetClassPath=&CommonTextBlock
   bIsVariable=true
   parent=ScaleBox_1
   widget=Text_Title
   properties {
    justification=Center
    text=Welcome to The Combine
    textTransformPolicy=ToUpper
    font {
     bForceMonospaced=false
     bMaterialIsStencil=false
     fontMaterial=null
     fontObject=&BurbankSmall
     letterSpacing=0
     monospacedWidth=1
     size=26
     skewAmount=0
     typefaceFontName=Black
     outlineSettings {
      bApplyOutlineToDropShadows=false
      bMiteredCorners=false
      bSeparateFillAlpha=false
      outlineMaterial=null
      outlineSize=1
      outlineColor=<rgba 0.0331 0.0203 0 0.8>
     }
    }
    shadowColorAndOpacity=<rgba 0.0902 0.0902 0.0902 1>
    shadowOffset=<4 4>
   }
   slot {class=ScaleBoxSlot}
  }
  SizeBox_0 {
   widgetClassPath=&SizeBox
   parent=VerticalBox_0
   widget=SizeBox_0
   properties {bOverride_MaxDesiredWidth=true, maxDesiredWidth=650}
   slot {class=VerticalBoxSlot, properties {padding=<ltrb 0 -7 0 0>}}
  }
  ScaleBox_0 {widgetClassPath=&ScaleBox, parent=SizeBox_0, widget=ScaleBox_0, slot {class=SizeBoxSlot}}
  HorizontalBox_1 {widgetClassPath=&HorizontalBox, parent=ScaleBox_0, widget=HorizontalBox_1, slot {class=ScaleBoxSlot}}
  Text_Description {
   widgetClassPath=&CommonTextBlock
   bIsVariable=true
   parent=HorizontalBox_1
   widget=Text_Description
   properties {
    justification=Center
    text=REMEMBER - emote to restart
    textTransformPolicy=ToUpper
    font {
     bForceMonospaced=false
     bMaterialIsStencil=false
     fontMaterial=null
     fontObject=&Burbank
     letterSpacing=0
     monospacedWidth=1
     size=16
     skewAmount=0
     typefaceFontName=Burbank-PlayerPower
     outlineSettings {
      bApplyOutlineToDropShadows=false
      bMiteredCorners=false
      bSeparateFillAlpha=false
      outlineMaterial=null
      outlineSize=1
      outlineColor=<rgba 0.0331 0.0203 0 0.8>
     }
    }
    shadowColorAndOpacity=<rgba 0.0902 0.0902 0.0902 1>
    shadowOffset=<4 4>
   }
   slot {class=HorizontalBoxSlot}
  }
  TutorialHighlight {widgetClassPath=&TutorialHighlight}
 ]
}
