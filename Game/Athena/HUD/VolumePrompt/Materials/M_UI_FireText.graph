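# Translucent UI material in the HUD VolumePrompt folder that draws animated, procedural fire, named for use with text.
# Two panning noise textures, one bump-offsetting the other, combine with radial gradients, one pulsing over time, to form the flame shape.
# Smoothstepped flame intensity blends two fire colors over a background color, with an optional white fill.
# A bump-offset mask texture and a radial outer shape each threshold the flame into cutout masks.
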Material /Game/Athena/HUD/VolumePrompt/Materials/M_UI_FireText
graph {
 properties {blendMode=BLEND_Translucent, materialDomain=MD_UI, pixelDepthOffsetMode=PDOM_Legacy, usedShadingModels=Default Lit}
 nodes [
  n1 = VectorParameter() {parameterName=BackgroundColor, defaultValue=<rgba 1 0.4125 0.003 1>}
  n2 = VectorParameter() {parameterName=Speed2, defaultValue=<rgba 0 0.5 0 1>}
  n3 = TextureSample(UVs=Panner(Speed=AppendVector(A=@n2.R, B=@n2.R)) {speedY=0.5}) {texture=/Game/Effects/Fort_Effects/Textures/Fire/T_FireNoise_02}
  n4 = VectorParameter() {parameterName=UVScale, defaultValue=<rgba 0 0 0 1>}
  n5 = Multiply(A=Multiply(A=TextureCoordinate(), B=AppendVector(A=@n4.R, B=@n4.R)), B=ScalarParameter() {defaultValue=1, parameterName=PatternScale})
  n6 = ScalarParameter() {defaultValue=2, parameterName=FlameDistort}
  n7 = VectorParameter() {parameterName=Speed1, defaultValue=<rgba 0 0.2 0 1>}
  n8 = TextureSample(UVs=Panner(
    Coordinate=MaterialFunctionCall!(HeightTexture=@n3.R, TextureCoordinates=@n5, ReferencePlane, HeightRatioInput=@n6, CameraVector) {materialFunction=/Engine/Functions/Engine_MaterialFunctions02/Utility/BumpOffset_Advanced.BumpOffset_advanced}.UVs
    Speed=AppendVector(A=@n7.R, B=@n7.R)
   ) {speedY=0.2}) {texture=/Game/Global/Textures/Default/Noise/T_FireNoise_01}
  n9 = Constant() {r=1}
  n10 = Constant() {r=2}
  n11 = MaterialFunctionCall(Value=Time()) {materialFunction=&LinearSine}
  n12 = Multiply(A=MaterialFunctionCall(
    UVs=@n5
    CenterPosition=Constant2Vector() {g=1, r=0.5}
    Radius=LinearInterpolate(A=@n9, B=@n10, Alpha=@n11.Linear Sine)
    Density=Constant() {r=0.7164}
   ) {materialFunction=&RadialGradientExponential}.RadialGradientExponential) {constB=-1}
  n13 = MaterialFunctionCall(UVs=@n5, Texture Scale=Constant2Vector() {g=2, r=3}) {materialFunction=&ScaleUVsByCenter}
  n14 = VectorParameter() {parameterName=FlameHeight, defaultValue=<rgba 0 0 0 1>}
  n15 = AppendVector(A=@n14.R, B=@n14.R)
  n16 = Multiply(
   A=LinearInterpolate(B=Power(Base=@n3.R) {constExponent=-0.75}, Alpha=Power(Base=@n8.R, Exp=@n12) {constExponent=5})
   B=MaterialFunctionCall(
    UVs=Add(A=@n13.UVs, B=@n15)
    CenterPosition=Constant2Vector() {g=1, r=0.5}
    Radius=Constant() {r=0.8043}
    Density=Constant() {r=2.3384}
    Invert Density=StaticBool() {value=true}
   ) {materialFunction=&RadialGradientExponential}.RadialGradientExponential
  )
  n17 = ScalarParameter() {defaultValue=2, parameterName=Firepower}
  n18 = MaterialFunctionCall(
   Alpha=@n16
   Min=ScalarParameter() {defaultValue=0.1, parameterName=MaskMin}
   Max=ScalarParameter() {defaultValue=0.76, parameterName=MaskMax}
  ) {materialFunction=&SmoothStep}
  n19 = Clamp(None=Power(Base=@n18.Result))
  n20 = LinearInterpolate(
   A=VectorParameter() {parameterName=Color2, defaultValue=<rgba 0.3281 0.0154 0 1>}.RGB
   B=VectorParameter() {parameterName=Color1, defaultValue=<rgba 1 0.8698 0.141 1>}.RGB
   Alpha=Multiply(A=Power(Base=@n16, Exp=@n17), B=@n19)
  )
  n21 = Multiply(A=@n19, B=ScalarParameter() {defaultValue=1, parameterName=FlameOpacity})
  LinearInterpolate(A=Clamp(None=LinearInterpolate(A=@n1.RGB, B=@n20, Alpha=@n21)), B=Constant() {r=1}, Alpha=ScalarParameter() {parameterName=WhiteFill})
  n22 = LinearInterpolate(Alpha=@n16)
  n23 = ScalarParameter() {defaultValue=0.5, group=FlameMask, parameterName=OutlineBumpAngle}
  n24 = ScalarParameter() {defaultValue=0.1, group=FlameMask, parameterName=OutlineDistort}
  n25 = VectorParameter() {group=FlameMask, parameterName=OutlineDistortVector, defaultValue=<rgba 0 0 0 1>}
  n26 = VectorParameter() {group=FlameMask, parameterName=MaskScale, defaultValue=<rgba 1 1 0 1>}
  n27 = TextureSampleParameter2D(UVs=MaterialFunctionCall(
    UVs=MaterialFunctionCall(HeightTexture=@n22, TextureCoordinates=@n5, ReferencePlane=@n23, HeightRatioInput=@n24, CameraVector=@n25.RGB) {materialFunction=&BumpOffset_advanced}.UVs
    Texture Scale=AppendVector(A=@n26.R, B=@n26.R)
   ) {materialFunction=&ScaleUVsByCenter}.UVs) {
   group=FlameMask
   parameterName=MaskTexture
   samplerSource=SSM_Clamp_WorldGroupSettings
   texture=/Game/Athena/HUD/LTM/Mash/Materials/T_UI_Mash_FlameEffect_Mask
  }
  n28 = OneMinus(None=@n27.R)
  n29 = ScalarParameter() {defaultValue=0.001, group=FlameMask, parameterName=OutlineSmooth}
  n30 = Multiply(A=@n29) {constB=0.1}
  MaterialFunctionCall(Alpha=@n18.Result, Min=@n28, Max=Add(A=@n28, B=LinearInterpolate(A=@n29, B=@n30, Alpha=@n11.Linear Sine)) {constB=0.01}) {materialFunction=&SmoothStep}
  MaterialFunctionCall() {materialFunction=&GetUserInterfaceUV}
  MaterialFunctionCall(HeightTexture=@n8.R, TextureCoordinates=@n5, HeightRatioInput=@n6) {materialFunction=&BumpOffset_advanced}
  n31 = VectorParameter() {group=OuterShape, parameterName=OuterShapePosition, defaultValue=<rgba 0.5 0.5 0 1>}
  n32 = OneMinus(None=MaterialFunctionCall(
    UVs=@n5
    CenterPosition=AppendVector(A=@n31.R, B=@n31.R)
    Radius=ScalarParameter() {defaultValue=0.5, group=OuterShape, parameterName=OuterShapeRadius}
    Density=ScalarParameter() {defaultValue=2.3, group=OuterShape, parameterName=OuterShapeDensity}
   ) {materialFunction=&RadialGradientExponential}.RadialGradientExponential)
  MaterialFunctionCall(Alpha=@n18.Result, Min=@n32, Max=Add(A=@n32) {constB=0.01}) {materialFunction=&SmoothStep}
  TextureCoordinate()
 ]
}
