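# Translucent UI material for a Moisty border in the HUD volume prompt, drawing a patterned shape with a drop shadow.
# A radial reveal gradient times a screen-aligned diamond pattern is smoothstepped, then masks border texture channels R or G, picked by Shape.
# An edge smoothstep blends ShadowColor into ShapeColor.
# A second, offset sample forms the shadow, maxed with the shape mask; its offset uses only ShadowOffset.R on both axes.
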Material /Game/Athena/HUD/VolumePrompt/Materials/M_UI_MoistyBorder
graph {
 properties {blendMode=BLEND_Translucent, materialDomain=MD_UI, pixelDepthOffsetMode=PDOM_Legacy, usedShadingModels=Default Lit}
 nodes [
  n1 = VectorParameter() {parameterName=ShadowColor, defaultValue=<rgba 1 0.0052 0.0056 0.4>}
  n2 = VectorParameter() {group=Reveal, parameterName=RevealCenter, defaultValue=<rgba 0.5 0.5 0 1>}
  n3 = ScalarParameter() {defaultValue=1, group=Reveal, parameterName=RevealRadius}
  n4 = ScalarParameter() {defaultValue=2, group=Reveal, parameterName=RevealDensity}
  n5 = TextureSample(UVs=Multiply(
    A=MaterialFunctionCall() {materialFunction=&ScreenAlignedUVs}."X Scale to Ratio, Y 100%"
    B=ScalarParameter() {defaultValue=1, group=Pattern, parameterName=PatternRepeat}
   )) {texture=/Game/Athena/HUD/Materials/Victory/T_Diamond-Pattern}
  n6 = ScalarParameter() {defaultValue=1, group=Pattern, parameterName=PatternPower}
  n7 = ScalarParameter() {group=Pattern, parameterName=PatternEdge}
  n8 = MaterialFunctionCall(
   Alpha=Multiply(
    A=MaterialFunctionCall(CenterPosition=@n2.RGB, Radius=@n3, Density=@n4) {materialFunction=&RadialGradientExponential}.RadialGradientExponential
    B=Power(Base=@n5.R, Exp=@n6)
   )
   Min=@n7
   Max=Add(A=@n7, B=ScalarParameter() {group=Pattern, parameterName=PatternSoftness})
  ) {materialFunction=&SmoothStep}
  n9 = ScalarParameter() {defaultValue=1, parameterName=Shape, sliderMax=1}
  n10 = Constant() {r=0.6}
  n11 = TextureSample() {samplerSource=SSM_Clamp_WorldGroupSettings, texture=&/T_UI_MoistyBorder}
  n12 = If(A=@n9, B=Constant() {r=0.3}, "A > B"=@n11.G, "A < B"=@n11.G)
  n13 = ScalarParameter() {defaultValue=0.2, parameterName=Edge}
  n14 = MaterialFunctionCall(
   Alpha=Multiply(A=@n8.Result, B=If(A=@n9, B=@n10, "A > B"=@n11.R, "A < B"=@n12))
   Min=@n13
   Max=Add(A=@n13, B=ScalarParameter() {defaultValue=0.01, parameterName=EdgeSoftness})
  ) {materialFunction=&SmoothStep}
  LinearInterpolate(A=@n1.RGB, B=VectorParameter() {parameterName=ShapeColor, defaultValue=<rgba 1 1 1 0.4>}.RGB, Alpha=@n14.Result)
  n15 = ScalarParameter() {defaultValue=1, parameterName=Shape, sliderMax=1}
  n16 = VectorParameter() {parameterName=ShadowOffset, defaultValue=<rgba -0.0154 -0.0238 0 1>}
  n17 = TextureSample(UVs=Add(A=TextureCoordinate(), B=AppendVector(A=@n16.R, B=@n16.R))) {samplerSource=SSM_Clamp_WorldGroupSettings, texture=&T_UI_MoistyBorder}
  n18 = Constant() {r=0.3}
  n19 = Multiply(A=@n8.Result, B=If(A=@n15, B=Constant() {r=0.6}, "A > B"=@n17.R, "A < B"=If(A=@n15, B=@n18, "A > B"=@n17.G, "A < B"=@n17.G)))
  n20 = ScalarParameter() {defaultValue=0.2, parameterName=Edge}
  n21 = Add(A=@n20, B=ScalarParameter() {defaultValue=0.01, parameterName=EdgeSoftness})
  Max(
   A=@n14.Result
   B=Multiply(
    A=MaterialFunctionCall(Alpha=@n19, Min=@n20, Max=@n21) {materialFunction=&SmoothStep}.Result
    B=@n1.A
   )
  )
 ]
}
