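# Translucent UI material for a volume prompt symbol, showing the Oak decal symbol with a Color tint and a noisy reveal mask.
# An inverted exponential radial gradient sets the SmoothStep threshold on a noise texture, with NoiseRevealSoftness widening the edge.
# NoiseFill is added to that mask, which is clamped and multiplied by the symbol's alpha, sampled at a fixed mip level.
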Material /Game/Athena/HUD/VolumePrompt/Materials/M_UI_OakPromptSymbol
graph {
 properties {blendMode=BLEND_Translucent, materialDomain=MD_UI, pixelDepthOffsetMode=PDOM_Legacy, usedShadingModels=Default Lit}
 nodes [
  n1 = OneMinus(None=MaterialFunctionCall(
    CenterPosition=VectorParameter() {parameterName=GradientCenter, defaultValue=<rgba 0.5 0 0 1>}.RGB
    Radius=ScalarParameter() {defaultValue=1, parameterName=Radius}
    Density=ScalarParameter() {defaultValue=1, parameterName=Density}
   ) {materialFunction=&RadialGradientExponential}.RadialGradientExponential)
  n2 = MaterialFunctionCall(
   Alpha=TextureSample() {texture=/Game/Global/Textures/Default/Noise/Noise}.R
   Min=@n1
   Max=Add(A=@n1, B=ScalarParameter() {parameterName=NoiseRevealSoftness})
  ) {materialFunction=&SmoothStep}
  n3 = ScalarParameter() {parameterName=NoiseFill}
  Multiply(
   A=Multiply(A=TextureSample!(UVs, Tex, Level, Apply View MipBias) {constMipValue=2, mipValueMode=TMVM_MipLevel, texture=/Game/Environments/Sets/Oak/Textures/Oak_Decal_Symbol}.A)
   B=Clamp(None=Add(A=@n2.Result, B=@n3))
  )
  VectorParameter() {parameterName=Color, defaultValue=<rgba 1 1 1 1>}
 ]
}
