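# UI-domain material for Oak volume prompt font text, producing a scratched, mottled two-color fill (yellow to orange by default).
# A scalable low-res blurred noise texture blends Color1 and Color2.
# A scratch texture, scaled around its center using only ScratchScale's red channel on both axes and shaped by a ScratchAmount power, blends that result toward a hardcoded dark brown.
# A screen-aligned UV branch feeds nothing.
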Material /Game/Athena/HUD/VolumePrompt/Materials/M_UI_OakPrompt_Font
graph {
 properties {materialDomain=MD_UI, pixelDepthOffsetMode=PDOM_Legacy, usedShadingModels=Default Lit}
 nodes [
  n1 = Multiply(A=TextureCoordinate(), B=ScalarParameter() {defaultValue=1, parameterName=NoiseScale})
  n2 = VectorParameter() {parameterName=ScratchScale, defaultValue=<rgba 1 1 0 1>}
  n3 = MaterialFunctionCall(Texture Scale=AppendVector(A=@n2.R, B=@n2.R)) {materialFunction=&ScaleUVsByCenter}
  LinearInterpolate(
   A=LinearInterpolate(
    A=VectorParameter() {parameterName=Color1, defaultValue=<rgba 0.9647 0.9911 0.0052 1>}.RGB
    B=VectorParameter() {parameterName=Color2, defaultValue=<rgba 0.5841 0.2159 0.006 1>}.RGB
    Alpha=TextureSample(UVs=@n1) {samplerType=SAMPLERTYPE_Grayscale, texture=&LowResBlurredNoise}.R
   )
   B=Constant3Vector() {constant=<rgba 0.2384 0.0343 0.0003 0>}
   Alpha=Power(
    Base=TextureSample(UVs=@n3.UVs) {texture=&/T_UI_OakPrompt_TextScratch}.R
    Exp=ScalarParameter() {defaultValue=1, parameterName=ScratchAmount}
   )
  )
  MaterialFunctionCall(Texture Dimensions=MaterialFunctionCall() {materialFunction=&GetUserInterfaceUV}.Tiling) {materialFunction=&ScreenAlignedUVs}
 ]
}
