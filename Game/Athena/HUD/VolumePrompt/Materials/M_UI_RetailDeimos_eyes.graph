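# UI-domain additive material in the Athena HUD VolumePrompt folder, named for the eyes of the RetailDeimos prompt art.
# The graph holds a single texture sample of T_UI_RetailDeimos_eyes, with no other math, parameters or animation.
# No link from that sample to a material output appears in the graph.
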Material /Game/Athena/HUD/VolumePrompt/Materials/M_UI_RetailDeimos_eyes
graph {
 properties {blendMode=BLEND_Additive, materialDomain=MD_UI, pixelDepthOffsetMode=PDOM_Legacy, usedShadingModels=Default Lit}
 nodes [TextureSample() {texture=&/T_UI_RetailDeimos_eyes}]
}
