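# Translucent UI material for the volume prompt HUD that draws an outlined, two-shape "Taco" border from a panning texture.
# PanSpeed's red value drives both pan axes and defaults to zero.
# SmoothStep thresholds on the R and G channels layer shape colors over an outline color, with G on top.
# A Max of lower-threshold steps on both channels forms the combined outlined silhouette mask.
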Material /Game/Athena/HUD/VolumePrompt/Materials/M_UI_TacoBorder
graph {
 properties {blendMode=BLEND_Translucent, materialDomain=MD_UI, pixelDepthOffsetMode=PDOM_Legacy, usedShadingModels=Default Lit}
 nodes [
  n1 = VectorParameter() {parameterName=PanSpeed, defaultValue=<rgba 0 0 0 1>}
  n2 = TextureSample(UVs=Panner(Speed=AppendVector(A=@n1.R, B=@n1.R))) {texture=&/T_UI_TacoBorder}
  n3 = ScalarParameter() {defaultValue=0.2, parameterName=Edge}
  n4 = Add(A=@n3, B=ScalarParameter() {defaultValue=0.01, parameterName=EdgeSoftness})
  n5 = ScalarParameter() {defaultValue=0.2, parameterName=Edge}
  n6 = ScalarParameter() {defaultValue=0.01, parameterName=EdgeSoftness}
  LinearInterpolate(
   A=LinearInterpolate(
    A=VectorParameter() {parameterName=OutlineColor, defaultValue=<rgba 0.0642 0.0398 0.1875 1>}.RGB
    B=VectorParameter() {parameterName=ShapeColor, defaultValue=<rgba 0.1946 0.6867 0.1529 1>}.RGB
    Alpha=MaterialFunctionCall(Alpha=@n2.R, Min=@n3, Max=@n4) {materialFunction=&SmoothStep}.Result
   )
   B=VectorParameter() {parameterName=ShapeColor, defaultValue=<rgba 0.7454 0.1144 0.0999 1>}.RGB
   Alpha=MaterialFunctionCall(Alpha=@n2.G, Min=@n5, Max=Add(A=@n5, B=@n6)) {materialFunction=&SmoothStep}.Result
  )
  n7 = ScalarParameter() {defaultValue=0.1, parameterName=OutlineSize}
  n8 = ScalarParameter() {defaultValue=0.01, parameterName=OutlineSoftness}
  n9 = ScalarParameter() {defaultValue=0.1, parameterName=OutlineSize}
  n10 = ScalarParameter() {defaultValue=0.01, parameterName=OutlineSoftness}
  Max(
   A=MaterialFunctionCall(Alpha=@n2.R, Min=@n7, Max=Add(A=@n7, B=@n8)) {materialFunction=&SmoothStep}.Result
   B=MaterialFunctionCall(Alpha=@n2.G, Min=@n9, Max=Add(A=@n9, B=@n10)) {materialFunction=&SmoothStep}.Result
  )
 ]
}
