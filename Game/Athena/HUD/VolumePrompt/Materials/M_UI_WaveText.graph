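# M_UI_WaveText is a UI material that produces a two-color wave pattern, apparently for text in the HUD volume prompt.
# A summer LTM sine texture, slightly tiled, pans diagonally because the Speed parameter's red channel drives both axes.
# SmoothStep thresholds its red channel using Edge and EdgeSoftness, and the result blends between the Color1 and Color2 parameters, white and yellow-green by default.
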Material /Game/Athena/HUD/VolumePrompt/Materials/M_UI_WaveText
graph {
 properties {materialDomain=MD_UI, pixelDepthOffsetMode=PDOM_Legacy, usedShadingModels=Default Lit}
 nodes [
  n1 = VectorParameter() {parameterName=Speed, defaultValue=<rgba 0 0 0 1>}
  n2 = Panner(Coordinate=TextureCoordinate() {vTiling=1.25}, Speed=AppendVector(A=@n1.R, B=@n1.R))
  n3 = ScalarParameter() {parameterName=Edge}
  n4 = ScalarParameter() {parameterName=EdgeSoftness}
  LinearInterpolate(
   A=VectorParameter() {parameterName=Color1, defaultValue=<rgba 1 1 1 1>}.RGB
   B=VectorParameter() {parameterName=Color2, defaultValue=<rgba 0.8879 1 0.2159 1>}.RGB
   Alpha=MaterialFunctionCall(Alpha=TextureSample(UVs=@n2) {texture=/Game/Athena/HUD/LTM/Summer/T_UI_Sin_DF}.R, Min=@n3, Max=Add(A=@n3, B=@n4)) {materialFunction=&SmoothStep}.Result
  )
 ]
}
