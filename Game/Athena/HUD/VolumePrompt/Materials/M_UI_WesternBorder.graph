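# Translucent UI material for the western-style border in the Athena HUD volume prompt, with a parameterized fill, outline and reveal.
# The border texture's red channel, sampled on 9-slice UVs, is multiplied by an inverted radial gradient that drives the reveal.
# SmoothStep thresholds on that mask separate fill from BorderColor and, at a lower OutlineEdge cutoff, form a wider outline shape.
# A second radial gradient tints the fill toward GradientColor.
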Material /Game/Athena/HUD/VolumePrompt/Materials/M_UI_WesternBorder
graph {
 properties {blendMode=BLEND_Translucent, materialDomain=MD_UI, pixelDepthOffsetMode=PDOM_Legacy, usedShadingModels=Default Lit}
 nodes [
  n1 = VectorParameter() {group=GradientColor, parameterName=GradientColor, defaultValue=<rgba 1 1 1 1>}
  n2 = MaterialFunctionCall(
   UVs=MaterialFunctionCall() {materialFunction=&GetUserInterfaceUV}.Normalized UV
   CenterPosition=VectorParameter() {group=GradientColor, parameterName=GradientPosition, defaultValue=<rgba 0 0 0 1>}.RGB
   Radius=ScalarParameter() {group=GradientColor, parameterName=GradientRadius}
  ) {materialFunction=&RadialGradientExponential}
  n3 = VectorParameter() {parameterName=TextureOffset, defaultValue=<rgba 0 0 0 1>}
  n4 = TextureSampleParameter2D(UVs=Add(A=MaterialFunctionCall() {materialFunction=&GetUserInterfaceUV}.9-Slice UV, B=AppendVector(A=@n3.R, B=@n3.R))) {parameterName=Texture, samplerSource=SSM_Clamp_WorldGroupSettings, texture=&/T_UI_WesternBorder}
  n5 = MaterialFunctionCall() {materialFunction=&GetUserInterfaceUV}
  n6 = VectorParameter() {group=Reveal, parameterName=RevealPosition, defaultValue=<rgba 0 0 0 1>}
  n7 = ScalarParameter() {group=Reveal, parameterName=RevealRadius}
  n8 = ScalarParameter() {group=Reveal, parameterName=RevealDensity}
  n9 = Multiply(
   A=@n4.R
   B=OneMinus(None=MaterialFunctionCall(UVs=@n5.Normalized UV, CenterPosition=@n6.RGB, Radius=@n7, Density=@n8) {materialFunction=&RadialGradientExponential}.RadialGradientExponential)
  )
  n10 = ScalarParameter() {defaultValue=0.2, parameterName=Edge}
  n11 = ScalarParameter() {defaultValue=0.01, parameterName=EdgeSoftness}
  LinearInterpolate(
   A=VectorParameter() {parameterName=BorderColor, defaultValue=<rgba 0.0331 0.0203 0 1>}.RGB
   B=LinearInterpolate(
    A=VectorParameter() {parameterName=Color, defaultValue=<rgba 0.8148 0.7157 0.4678 1>}.RGB
    B=@n1.RGB
    Alpha=Multiply(A=@n2.RadialGradientExponential, B=@n1.A)
   )
   Alpha=MaterialFunctionCall(Alpha=@n9, Min=@n10, Max=Add(A=@n10, B=@n11)) {materialFunction=&SmoothStep}.Result
  )
  n12 = Subtract(A=@n10, B=ScalarParameter() {defaultValue=0.1, parameterName=OutlineEdge})
  MaterialFunctionCall(Alpha=@n9, Min=@n12, Max=Add(A=@n12, B=ScalarParameter() {defaultValue=0.01, parameterName=EdgeSoftness})) {materialFunction=&SmoothStep}
 ]
}
