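# HUD volume prompt widget styled for Moisty Palms, showing a title and description, by default "Crouch to prop!", with a static "Fabulous" tag.
# OpenWidget sets both texts from the prompt data, rebinds the Outro's finished event, then plays Intro.
# Intro staggers text pops, fades and slides with material-parameter reveals on the backplate, diamonds and stars.
# After a retriggerable seven-second delay Outro fades the container, while CloseWidget collapses the widget.
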WidgetBlueprint /Game/Athena/HUD/VolumePrompt/MoistyPropPrompt
animations [
 Intro {
  binding_count=10
  length_seconds=1.4395
  object=&self:Intro
  sequence {
   bindings [
    Image_star1 {
     id=#1
     possessed_class=Image
     tracks [
      {
       type=MovieSceneWidgetMaterialTrack
       sections [{
         type=MovieSceneParameterSection
         channels [
          RevealDensity {type=MovieSceneScriptingFloatChannel, keys [[50.793,0],[84.578,4.0649]]}
          RevealRadius {type=MovieSceneScriptingFloatChannel, keys [[36.053,0],[69.837,0.6033]]}
          PatternPower {type=MovieSceneScriptingFloatChannel, keys [[56.437,0.667],[86.366,0]]}
         ]
        }]
      }
      {
       type=MovieScene2DTransformTrack
       sections [{
         type=MovieScene2DTransformSection
         channels [
          Translation.X {type=MovieSceneScriptingFloatChannel}
          Translation.Y {type=MovieSceneScriptingFloatChannel}
          Angle {type=MovieSceneScriptingFloatChannel}
          Scale.X {type=MovieSceneScriptingFloatChannel, keys [[36.248,1],[77.448,1.34],[82.326,1]]}
          Scale.Y {type=MovieSceneScriptingFloatChannel, keys [[36.248,1],[77.448,1.34],[82.326,1]]}
          Shear.X {type=MovieSceneScriptingFloatChannel}
          Shear.Y {type=MovieSceneScriptingFloatChannel}
         ]
        }]
      }
     ]
    }
    Image_Backplate {
     id=#2
     possessed_class=Image
     tracks [{
       type=MovieSceneWidgetMaterialTrack
       sections [{
         type=MovieSceneParameterSection
         channels [
          RevealDensity {type=MovieSceneScriptingFloatChannel, keys [[14.74,0],[48.525,4.0649]]}
          RevealRadius {type=MovieSceneScriptingFloatChannel, keys [[0,0],[33.784,0.6033]]}
          PatternPower {type=MovieSceneScriptingFloatChannel, keys [[33.784,0.667],[50.313,-1.9385]]}
         ]
        }]
      }]
    }
    Text_Title {
     id=#3
     possessed_class=CommonTextBlock
     tracks [
      {
       type=MovieSceneFloatTrack
       display_name=Render Opacity
       sections [{type=MovieSceneFloatSection, channels [None {type=MovieSceneScriptingFloatChannel, keys [[22.553,0],[38.117,1]]}]}]
      }
      {
       type=MovieScene2DTransformTrack
       sections [{
         type=MovieScene2DTransformSection
         channels [
          Scale.X {type=MovieSceneScriptingFloatChannel, keys [[22.553,2],[38.812,0.9535],[43.284,1]]}
          Scale.Y {type=MovieSceneScriptingFloatChannel, keys [[22.553,2],[38.812,0.9535],[43.284,1]]}
         ]
        }]
      }
     ]
    }
    Image_diamond2 {
     id=#4
     possessed_class=Image
     tracks [
      {
       type=MovieSceneWidgetMaterialTrack
       sections [{
         type=MovieSceneParameterSection
         channels [
          RevealDensity {type=MovieSceneScriptingFloatChannel, keys [[20.202,0],[43.664,4.0649]]}
          RevealRadius {type=MovieSceneScriptingFloatChannel, keys [[9.967,0],[33.427,0.6033]]}
          PatternPower {type=MovieSceneScriptingFloatChannel, keys [[24.121,0.667],[44.906,0]]}
         ]
        }]
      }
      {
       type=MovieScene2DTransformTrack
       sections [{
         type=MovieScene2DTransformSection
         channels [
          Scale.X {type=MovieSceneScriptingFloatChannel, keys [[9.967,1],[39.761,1.0956],[44.213,1]]}
          Scale.Y {type=MovieSceneScriptingFloatChannel, keys [[9.967,1],[39.761,1.0956],[44.213,1]]}
         ]
        }]
      }
     ]
    }
    Text_Description {
     id=#5
     possessed_class=CommonTextBlock
     tracks [
      {
       type=MovieSceneFloatTrack
       display_name=Render Opacity
       sections [{type=MovieSceneFloatSection, channels [None {type=MovieSceneScriptingFloatChannel, keys [[44.303,0],[60.033,1]]}]}]
      }
      {
       type=MovieScene2DTransformTrack
       sections [{
         type=MovieScene2DTransformSection
         channels [
          Translation.X {type=MovieSceneScriptingFloatChannel, keys [[44.303,-176.114],[55.197,-1.3553],[56.581,0]]}
          Translation.Y {type=MovieSceneScriptingFloatChannel}
         ]
        }]
      }
     ]
    }
    Image_star2 {
     id=#6
     possessed_class=Image
     tracks [
      {
       type=MovieSceneWidgetMaterialTrack
       sections [{
         type=MovieSceneParameterSection
         channels [
          RevealDensity {type=MovieSceneScriptingFloatChannel, keys [[37.934,0],[71.721,4.0649]]}
          RevealRadius {type=MovieSceneScriptingFloatChannel, keys [[23.193,0],[56.979,0.6033]]}
          PatternPower {type=MovieSceneScriptingFloatChannel, keys [[43.579,0.667],[73.508,0]]}
         ]
        }]
      }
      {
       type=MovieScene2DTransformTrack
       sections [{
         type=MovieScene2DTransformSection
         channels [
          Translation.X {type=MovieSceneScriptingFloatChannel}
          Translation.Y {type=MovieSceneScriptingFloatChannel}
          Angle {type=MovieSceneScriptingFloatChannel}
          Scale.X {type=MovieSceneScriptingFloatChannel, keys [[23.193,1],[64.393,1.34],[69.271,1]]}
          Scale.Y {type=MovieSceneScriptingFloatChannel, keys [[23.193,1],[64.393,1.34],[69.271,1]]}
          Shear.X {type=MovieSceneScriptingFloatChannel}
          Shear.Y {type=MovieSceneScriptingFloatChannel}
         ]
        }]
      }
     ]
    }
    Overlay_Container {
     id=#7
     possessed_class=Overlay
     tracks [{
       type=MovieSceneFloatTrack
       display_name=Render Opacity
       sections [{type=MovieSceneFloatSection, channels [None {type=MovieSceneScriptingFloatChannel, keys [[0,0],[33.993,1]]}]}]
      }]
    }
    Image_diamond3 {
     id=#8
     possessed_class=Image
     tracks [
      {
       type=MovieSceneWidgetMaterialTrack
       sections [{
         type=MovieSceneParameterSection
         channels [
          RevealDensity {type=MovieSceneScriptingFloatChannel, keys [[27.402,0],[50.864,4.0649]]}
          RevealRadius {type=MovieSceneScriptingFloatChannel, keys [[17.166,0],[40.627,0.6033]]}
          PatternPower {type=MovieSceneScriptingFloatChannel, keys [[31.321,0.667],[52.106,0]]}
         ]
        }]
      }
      {
       type=MovieScene2DTransformTrack
       sections [{
         type=MovieScene2DTransformSection
         channels [
          Scale.X {type=MovieSceneScriptingFloatChannel, keys [[17.166,1],[46.96,1.0956],[51.412,1]]}
          Scale.Y {type=MovieSceneScriptingFloatChannel, keys [[17.166,1],[46.96,1.0956],[51.412,1]]}
         ]
        }]
      }
     ]
    }
    Text_fabulous {
     id=#9
     possessed_class=CommonTextBlock
     tracks [{
       type=MovieScene2DTransformTrack
       sections [{
         type=MovieScene2DTransformSection
         channels [
          Translation.X {type=MovieSceneScriptingFloatChannel}
          Translation.Y {type=MovieSceneScriptingFloatChannel}
          Angle {type=MovieSceneScriptingFloatChannel}
          Scale.X {type=MovieSceneScriptingFloatChannel, keys [[34.257,0],[41.723,1.8043],[47.927,0.9308],[48.91,1]]}
          Scale.Y {type=MovieSceneScriptingFloatChannel, keys [[34.257,0],[41.723,1.8042],[47.927,0.9308],[48.91,1]]}
          Shear.X {type=MovieSceneScriptingFloatChannel}
          Shear.Y {type=MovieSceneScriptingFloatChannel}
         ]
        }]
      }]
    }
    Image_diamond1 {
     id=#10
     possessed_class=Image
     tracks [
      {
       type=MovieSceneWidgetMaterialTrack
       sections [{
         type=MovieSceneParameterSection
         channels [
          RevealDensity {type=MovieSceneScriptingFloatChannel, keys [[27.402,0],[50.864,4.0649]]}
          RevealRadius {type=MovieSceneScriptingFloatChannel, keys [[17.166,0],[40.627,0.6033]]}
          PatternPower {type=MovieSceneScriptingFloatChannel, keys [[31.321,0.667],[52.106,0]]}
         ]
        }]
      }
      {
       type=MovieScene2DTransformTrack
       sections [{
         type=MovieScene2DTransformSection
         channels [
          Scale.X {type=MovieSceneScriptingFloatChannel, keys [[17.166,1],[46.96,1.0956],[51.412,1]]}
          Scale.Y {type=MovieSceneScriptingFloatChannel, keys [[17.166,1],[46.96,1.0956],[51.412,1]]}
         ]
        }]
      }
     ]
    }
   ]
   display_rate {numerator=60}
  }
 }
 Outro {
  binding_count=1
  length_seconds=0.7391
  object=&self:Outro
  sequence {bindings [Overlay_Container {
     id=#11
     possessed_class=Overlay
     tracks [{
       type=MovieSceneFloatTrack
       display_name=Render Opacity
       sections [{type=MovieSceneFloatSection, end_frame=14, start_frame=0, channels [None {type=MovieSceneScriptingFloatChannel, keys [[0,1],[14.7813,0]]}]}]
      }]
    }]}
 }
]
graphs [
 EventGraph {nodes [
   Event CloseWidget() {is=title K2Node_Event}
   > Parent: CloseWidget() {is=title K2Node_CallParentFunction}
   > SetVisibility(InVisibility=Collapsed)
   n1 = Event OpenWidget!() -> (OutputDelegate, then, PrompData:"Fort HUDTag Prompt Data Structure (by ref)") {is=title K2Node_Event}
   > Parent: OpenWidget!(execute, PrompData:"Fort HUDTag Prompt Data Structure (by ref)"=@n1.PrompData) -> (then) {is=title K2Node_CallParentFunction}
   > K2Node_Composite_0 = Set Text and Images!(execute, VolumePromptData:"Fort HUDTag Prompt Data Structure (by ref)"=@n1.PrompData) -> (then:Exec) {is=title K2Node_Composite}
   > UnbindAllFromAnimationFinished!(
    execute
    self:User Widget Object Reference
    Animation:Widget Animation Object Reference=Get Outro() -> (Outro:Widget Animation Object Reference)
   ) -> (then)
   > StopAnimation(InAnimation=Get Outro())
   > BindToAnimationFinished(Animation=Get Outro(), Delegate=Create Event())
   > SetVisibility(InVisibility=HitTestInvisible)
   > PlayAnimationForward(InAnimation=Get Intro())
   > RetriggerableDelay(Duration=7)
   > PlayAnimation(InAnimation=Get Outro())
  ]}
 Set Text and Images {
  path=EventGraph.K2Node_Composite_0.Set Text and Images
  nodes [
   n1 = Inputs!() -> (execute, VolumePromptData:"Fort HUDTag Prompt Data Structure (by ref)")
   n2 = Break Fort HUDTag Prompt Data!(FortHUDTagPromptData:"Fort HUDTag Prompt Data Structure (by ref)"=@n1) -> (
    VolumeTag:Gameplay Tag Structure
    Text_PromptTitle:Text
    Text_PromptDescription:Text
    bIsEnterData:bool=~
    WidgetClass:User Widget Class Reference
   ) {is=title K2Node_BreakStruct}
   > "SetText (Text)"(self=Get Text_Title() -> (Text_Title:Common Text Object Reference), InText=@n2.Text_PromptTitle)
   > "SetText (Text)"(self=Get Text_Description() -> (Text_Description:Common Text Object Reference), InText=@n2.Text_PromptDescription)
   > Outputs!(then:Exec)
  ]
 }
]
variables []
widgets {
 info {inheritedWidgetCount=1, parentClass=/Script/FortniteUI.FortHUDTagPromptWidgetBase, rootWidgetClass=&Overlay, widgetCount=22}
 widgets [
  Overlay_Container {widgetClassPath=&Overlay, widget=Overlay_Container}
  Image_Shadow {
   widgetClassPath=&Image
   bIsVariable=true
   parent=Overlay_Container
   widget=Image_Shadow
   properties {
    brush {
     bIsDynamicallyLoaded=false
     drawAs=Image
     imageType=NoImage
     mirroring=NoMirror
     resourceName=None
     resourceObject=/Game/Athena/HUD/LTM/Wax/Materials/M_UI_TextGlow
     tiling=NoTile
     imageSize=<1600 400>
     margin=<ltrb 0 0 0 0>
     outlineSettings {
      bUseBrushTransparency=false
      roundingType=HalfHeightRadius
      width=0
      color {colorUseRule=UseColor_Specified, specifiedColor=<rgba 0 0 0 0>}
      cornerRadii=<0 0 0 0>
     }
     tintColor {colorUseRule=UseColor_Specified, specifiedColor=<rgba 1 1 1 1>}
     uVRegion {bIsValid=false, max=<0 0>, min=<0 0>}
    }
    colorAndOpacity=<rgba 0.0312 0.0202 0.0155 0.8>
   }
   slot {class=OverlaySlot, properties {horizontalAlignment=HAlign_Center, verticalAlignment=VAlign_Center, padding=<ltrb -800 -200 -800 -200>}}
  }
  Overlay_0 {
   widgetClassPath=&Overlay
   parent=Overlay_Container
   widget=Overlay_0
   slot {class=OverlaySlot, properties {horizontalAlignment=HAlign_Fill, verticalAlignment=VAlign_Fill, padding=<ltrb -40 0 -40 0>}}
  }
  Overlay_3 {
   widgetClassPath=&Overlay
   parent=Overlay_0
   widget=Overlay_3
   slot {class=OverlaySlot, properties {horizontalAlignment=HAlign_Center, verticalAlignment=VAlign_Center}}
  }
  Image_Backplate {
   widgetClassPath=&Image
   bIsVariable=true
   parent=Overlay_3
   widget=Image_Backplate
   properties {brush {
     bIsDynamicallyLoaded=false
     drawAs=Image
     imageType=NoImage
     mirroring=NoMirror
     resourceName=None
     resourceObject=/Game/Athena/HUD/VolumePrompt/Materials/M_UI_MoistyBorder_Inst
     tiling=NoTile
     imageSize=<386 167>
     margin=<ltrb 0 0 0 0>
     outlineSettings {
      bUseBrushTransparency=false
      roundingType=HalfHeightRadius
      width=0
      color {colorUseRule=UseColor_Specified, specifiedColor=<rgba 0 0 0 0>}
      cornerRadii=<0 0 0 0>
     }
     tintColor {colorUseRule=UseColor_Specified, specifiedColor=<rgba 1 1 1 1>}
     uVRegion {bIsValid=false, max=<0 0>, min=<0 0>}
    }}
   slot {class=OverlaySlot}
  }
  HorizontalBox_3 {
   widgetClassPath=&HorizontalBox
   parent=Overlay_3
   widget=HorizontalBox_3
   slot {class=OverlaySlot, properties {horizontalAlignment=HAlign_Center, verticalAlignment=VAlign_Center}}
  }
  Image_diamond1 {
   widgetClassPath=&Image
   parent=HorizontalBox_3
   widget=Image_diamond1
   properties {brush {
     bIsDynamicallyLoaded=false
     drawAs=Image
     imageType=NoImage
     mirroring=NoMirror
     resourceName=None
     resourceObject=/Game/Athena/HUD/VolumePrompt/Materials/M_UI_MoistyDiamond_Inst
     tiling=NoTile
     imageSize=<86 169>
     margin=<ltrb 0 0 0 0>
     outlineSettings {
      bUseBrushTransparency=false
      roundingType=HalfHeightRadius
      width=0
      color {colorUseRule=UseColor_Specified, specifiedColor=<rgba 0 0 0 0>}
      cornerRadii=<0 0 0 0>
     }
     tintColor {colorUseRule=UseColor_Specified, specifiedColor=<rgba 1 1 1 1>}
     uVRegion {bIsValid=false, max=<0 0>, min=<0 0>}
    }}
   slot {class=HorizontalBoxSlot}
  }
  Image_diamond2 {
   widgetClassPath=&Image
   parent=HorizontalBox_3
   widget=Image_diamond2
   properties {brush {
     bIsDynamicallyLoaded=false
     drawAs=Image
     imageType=NoImage
     mirroring=NoMirror
     resourceName=None
     resourceObject=&M_UI_MoistyDiamond_Inst
     tiling=NoTile
     imageSize=<86 169>
     margin=<ltrb 0 0 0 0>
     outlineSettings {
      bUseBrushTransparency=false
      roundingType=HalfHeightRadius
      width=0
      color {colorUseRule=UseColor_Specified, specifiedColor=<rgba 0 0 0 0>}
      cornerRadii=<0 0 0 0>
     }
     tintColor {colorUseRule=UseColor_Specified, specifiedColor=<rgba 1 1 1 1>}
     uVRegion {bIsValid=false, max=<0 0>, min=<0 0>}
    }}
   slot {class=HorizontalBoxSlot}
  }
  Image_diamond3 {
   widgetClassPath=&Image
   parent=HorizontalBox_3
   widget=Image_diamond3
   properties {brush {
     bIsDynamicallyLoaded=false
     drawAs=Image
     imageType=NoImage
     mirroring=NoMirror
     resourceName=None
     resourceObject=&M_UI_MoistyDiamond_Inst
     tiling=NoTile
     imageSize=<86 169>
     margin=<ltrb 0 0 0 0>
     outlineSettings {
      bUseBrushTransparency=false
      roundingType=HalfHeightRadius
      width=0
      color {colorUseRule=UseColor_Specified, specifiedColor=<rgba 0 0 0 0>}
      cornerRadii=<0 0 0 0>
     }
     tintColor {colorUseRule=UseColor_Specified, specifiedColor=<rgba 1 1 1 1>}
     uVRegion {bIsValid=false, max=<0 0>, min=<0 0>}
    }}
   slot {class=HorizontalBoxSlot}
  }
  Image_star1 {
   widgetClassPath=&Image
   bIsVariable=true
   parent=Overlay_3
   widget=Image_star1
   properties {
    brush {
     bIsDynamicallyLoaded=false
     drawAs=Image
     imageType=NoImage
     mirroring=NoMirror
     resourceName=None
     resourceObject=/Game/Athena/HUD/VolumePrompt/Materials/M_UI_MoistyStar_Inst
     tiling=NoTile
     imageSize=<47 59>
     margin=<ltrb 0 0 0 0>
     outlineSettings {
      bUseBrushTransparency=false
      roundingType=HalfHeightRadius
      width=0
      color {colorUseRule=UseColor_Specified, specifiedColor=<rgba 0 0 0 0>}
      cornerRadii=<0 0 0 0>
     }
     tintColor {colorUseRule=UseColor_Specified, specifiedColor=<rgba 1 1 1 1>}
     uVRegion {bIsValid=false, max=<0 0>, min=<0 0>}
    }
    renderTransform {angle=0, scale=<1 1>, shear=<0 0>, translation=<41 -15>}
   }
   slot {class=OverlaySlot, properties {verticalAlignment=VAlign_Bottom}}
  }
  Image_star2 {
   widgetClassPath=&Image
   bIsVariable=true
   parent=Overlay_3
   widget=Image_star2
   properties {
    brush {
     bIsDynamicallyLoaded=false
     drawAs=Image
     imageType=NoImage
     mirroring=NoMirror
     resourceName=None
     resourceObject=&M_UI_MoistyStar_Inst
     tiling=NoTile
     imageSize=<26 34>
     margin=<ltrb 0 0 0 0>
     outlineSettings {
      bUseBrushTransparency=false
      roundingType=HalfHeightRadius
      width=0
      color {colorUseRule=UseColor_Specified, specifiedColor=<rgba 0 0 0 0>}
      cornerRadii=<0 0 0 0>
     }
     tintColor {colorUseRule=UseColor_Specified, specifiedColor=<rgba 1 1 1 1>}
     uVRegion {bIsValid=false, max=<0 0>, min=<0 0>}
    }
    renderTransform {angle=0, scale=<1 1>, shear=<0 0>, translation=<-39 25>}
   }
   slot {class=OverlaySlot, properties {horizontalAlignment=HAlign_Right}}
  }
  Overlay_1 {
   widgetClassPath=&Overlay
   parent=Overlay_Container
   widget=Overlay_1
   slot {class=OverlaySlot, properties {verticalAlignment=VAlign_Center}}
  }
  VerticalBox_0 {widgetClassPath=&VerticalBox, parent=Overlay_1, widget=VerticalBox_0, slot {class=OverlaySlot}}
  SizeBox_1 {
   widgetClassPath=&SizeBox
   parent=VerticalBox_0
   widget=SizeBox_1
   properties {bOverride_MaxDesiredWidth=true, maxDesiredWidth=650}
   slot {class=VerticalBoxSlot}
  }
  ScaleBox_1 {widgetClassPath=&ScaleBox, parent=SizeBox_1, widget=ScaleBox_1, properties {stretchDirection=DownOnly}, slot {class=SizeBoxSlot}}
  Text_Title {
   widgetClassPath=&CommonTextBlock
   bIsVariable=true
   parent=ScaleBox_1
   widget=Text_Title
   properties {
    justification=Center
    text=Moisty Palms
    textTransformPolicy=ToUpper
    font {
     bForceMonospaced=false
     bMaterialIsStencil=false
     fontMaterial=/Game/Athena/HUD/VolumePrompt/Materials/M_UI_WaveText_Inst
     fontObject=&Burbank
     letterSpacing=0
     monospacedWidth=1
     size=35
     skewAmount=0
     typefaceFontName=BurbankRegular-Black
     outlineSettings {
      bApplyOutlineToDropShadows=false
      bMiteredCorners=false
      bSeparateFillAlpha=false
      outlineMaterial=null
      outlineSize=1
      outlineColor=<rgba 0.0331 0.006 0 0.8>
     }
    }
    renderTransform {angle=0, scale=<1 1>, shear=<-8 0>, translation=<0 0>}
    shadowColorAndOpacity=<rgba 0.1301 0 0.0015 1>
    shadowOffset=<5 5>
   }
   slot {class=ScaleBoxSlot}
  }
  SizeBox_0 {
   widgetClassPath=&SizeBox
   parent=VerticalBox_0
   widget=SizeBox_0
   properties {bOverride_MaxDesiredWidth=true, maxDesiredWidth=650}
   slot {class=VerticalBoxSlot, properties {padding=<ltrb 0 -7 0 0>}}
  }
  ScaleBox_0 {widgetClassPath=&ScaleBox, parent=SizeBox_0, widget=ScaleBox_0, slot {class=SizeBoxSlot}}
  HorizontalBox_1 {
   widgetClassPath=&HorizontalBox
   parent=ScaleBox_0
   widget=HorizontalBox_1
   slot {class=ScaleBoxSlot, properties {horizontalAlignment=HAlign_Right}}
  }
  Text_Description {
   widgetClassPath=&CommonTextBlock
   bIsVariable=true
   parent=HorizontalBox_1
   widget=Text_Description
   properties {
    justification=Center
    text="Crouch to prop!"
    textTransformPolicy=ToUpper
    font {
     bForceMonospaced=false
     bMaterialIsStencil=false
     fontMaterial=null
     fontObject=&BurbankSmall
     letterSpacing=0
     monospacedWidth=1
     size=16
     skewAmount=0
     typefaceFontName=Bold
     outlineSettings {
      bApplyOutlineToDropShadows=false
      bMiteredCorners=false
      bSeparateFillAlpha=false
      outlineMaterial=null
      outlineSize=1
      outlineColor=<rgba 0.0331 0.006 0 0.8>
     }
    }
    renderTransform {angle=0, scale=<1 1>, shear=<0 0>, translation=<0 -7>}
    shadowColorAndOpacity=<rgba 0.1301 0 0.0015 1>
    shadowOffset=<3 3>
   }
   slot {class=HorizontalBoxSlot, properties {horizontalAlignment=HAlign_Right}}
  }
  Text_fabulous {
   widgetClassPath=&CommonTextBlock
   bIsVariable=true
   parent=Overlay_1
   widget=Text_fabulous
   properties {
    justification=Center
    text=Fabulous
    textTransformPolicy=ToUpper
    font {
     bForceMonospaced=false
     bMaterialIsStencil=false
     fontMaterial=null
     fontObject=&Burbank
     letterSpacing=0
     monospacedWidth=1
     size=16
     skewAmount=0
     typefaceFontName=Burbank-PlayerPower
     outlineSettings {
      bApplyOutlineToDropShadows=false
      bMiteredCorners=false
      bSeparateFillAlpha=false
      outlineMaterial=null
      outlineSize=1
      outlineColor=<rgba 0.0331 0.006 0 0.8>
     }
    }
    renderTransform {angle=-10, scale=<1 1>, shear=<-12 0>, translation=<-37 -25>}
    renderTransformPivot=<1.2 1.2>
    shadowColorAndOpacity=<rgba 0.1301 0 0.0015 1>
    shadowOffset=<3 3>
   }
   slot {class=OverlaySlot}
  }
  TutorialHighlight {widgetClassPath=&TutorialHighlight}
 ]
}
