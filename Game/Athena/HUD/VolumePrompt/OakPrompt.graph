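# HUD volume prompt widget, derived from FortHUDTagPromptWidgetBase, showing a stylized location banner with "Welcome to Pandora" placeholder text.
# OpenWidget fills title and description from prompt data; the "Shields Up, Traveler!" tagline stays static.
# It plays Intro, which pops in text, reveals the symbol through animated material parameters and slides in the tagline.
# A retriggerable seven-second delay then plays the Outro fade, with a bound finish callback; CloseWidget collapses the widget.
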WidgetBlueprint /Game/Athena/HUD/VolumePrompt/OakPrompt
animations [
 Intro {
  binding_count=5
  length_seconds=1.8173
  object=&self:Intro
  sequence {bindings [
    Text_Title_1 {
     id=#1
     possessed_class=CommonTextBlock
     tracks [
      {
       type=MovieSceneFloatTrack
       display_name=Render Opacity
       sections [{type=MovieSceneFloatSection, channels [None {type=MovieSceneScriptingFloatChannel, keys [[33.9983,0],[35.754,1]]}]}]
      }
      {
       type=MovieScene2DTransformTrack
       sections [{type=MovieScene2DTransformSection, channels [Translation.X {type=MovieSceneScriptingFloatChannel, keys [[33.9983,-157.898],[36.345,0]]}]}]
      }
     ]
    }
    Text_Title {
     id=#2
     possessed_class=CommonTextBlock
     tracks [
      {
       type=MovieScene2DTransformTrack
       sections [{
         type=MovieScene2DTransformSection
         channels [
          Translation.X {type=MovieSceneScriptingFloatChannel}
          Translation.Y {type=MovieSceneScriptingFloatChannel}
          Angle {type=MovieSceneScriptingFloatChannel}
          Scale.X {type=MovieSceneScriptingFloatChannel, keys [[3.8403,3.9947],[8.126,0.9676],[8.9367,1.0238],[10.4383,0.9866],[12.3697,1]]}
          Scale.Y {type=MovieSceneScriptingFloatChannel, keys [[3.8403,3.9947],[8.126,0.9676],[8.9367,1.0238],[10.4383,0.9866],[12.3697,1]]}
          Shear.X {type=MovieSceneScriptingFloatChannel}
          Shear.Y {type=MovieSceneScriptingFloatChannel}
         ]
        }]
      }
      {
       type=MovieSceneFloatTrack
       display_name=Render Opacity
       sections [{type=MovieSceneFloatSection, channels [None {type=MovieSceneScriptingFloatChannel, keys [[3.8403,0],[7.2127,1]]}]}]
      }
     ]
    }
    Text_Description {
     id=#3
     possessed_class=CommonTextBlock
     tracks [
      {
       type=MovieSceneFloatTrack
       display_name=Render Opacity
       sections [{type=MovieSceneFloatSection, channels [None {type=MovieSceneScriptingFloatChannel, keys [[3.8583,0],[9.352,1]]}]}]
      }
      {
       type=MovieScene2DTransformTrack
       sections [{
         type=MovieScene2DTransformSection
         channels [
          Translation.X {type=MovieSceneScriptingFloatChannel}
          Translation.Y {type=MovieSceneScriptingFloatChannel}
          Angle {type=MovieSceneScriptingFloatChannel}
          Scale.X {type=MovieSceneScriptingFloatChannel, keys [[3.8583,3.6285],[7.5453,0.9771],[8.0343,1.1408],[8.9453,0.9478],[9.9807,1]]}
          Scale.Y {type=MovieSceneScriptingFloatChannel, keys [[3.8583,3.6285],[7.5453,0.9771],[8.0343,1.1408],[8.9453,0.9478],[9.9807,1]]}
          Shear.X {type=MovieSceneScriptingFloatChannel}
          Shear.Y {type=MovieSceneScriptingFloatChannel}
         ]
        }]
      }
     ]
    }
    Overlay_Container {
     id=#4
     possessed_class=Overlay
     tracks [
      {
       type=MovieSceneFloatTrack
       display_name=Render Opacity
       sections [{type=MovieSceneFloatSection, channels [None {type=MovieSceneScriptingFloatChannel, keys [[0,0],[3.691,1]]}]}]
      }
      {
       type=MovieScene2DTransformTrack
       sections [{
         type=MovieScene2DTransformSection
         channels [
          Translation.X {type=MovieSceneScriptingFloatChannel}
          Translation.Y {type=MovieSceneScriptingFloatChannel}
          Angle {type=MovieSceneScriptingFloatChannel}
          Scale.X {type=MovieSceneScriptingFloatChannel, keys [[29.9977,1],[33.9983,0.75]]}
          Scale.Y {type=MovieSceneScriptingFloatChannel, keys [[29.9977,1],[33.9983,0.75]]}
          Shear.X {type=MovieSceneScriptingFloatChannel}
          Shear.Y {type=MovieSceneScriptingFloatChannel}
         ]
        }]
      }
     ]
    }
    Image_Symbol {
     id=#5
     possessed_class=Image
     tracks [{
       type=MovieSceneWidgetMaterialTrack
       sections [{
         type=MovieSceneParameterSection
         channels [
          Radius {type=MovieSceneScriptingFloatChannel, keys [[4.189,0],[21.522,1.7467]]}
          Density {type=MovieSceneScriptingFloatChannel, keys [[4.189,0],[14.4257,6.3305]]}
          NoiseFill {type=MovieSceneScriptingFloatChannel, keys [[10.1117,0],[21.522,1]]}
          NoiseRevealSoftness {type=MovieSceneScriptingFloatChannel, keys [[2.0737,0.2],[21.522,0]]}
         ]
        }]
      }]
    }
   ]}
 }
 Outro {
  binding_count=1
  length_seconds=0.7391
  object=&self:Outro
  sequence {bindings [Overlay_Container {
     id=#6
     possessed_class=Overlay
     tracks [{
       type=MovieSceneFloatTrack
       display_name=Render Opacity
       sections [{type=MovieSceneFloatSection, end_frame=14, start_frame=0, channels [None {type=MovieSceneScriptingFloatChannel, keys [[0,1],[14.7813,0]]}]}]
      }]
    }]}
 }
]
graphs [
 EventGraph {nodes [
   Event CloseWidget() {is=title K2Node_Event}
   > Parent: CloseWidget() {is=title K2Node_CallParentFunction}
   > SetVisibility(InVisibility=Collapsed)
   n1 = Event OpenWidget!() -> (OutputDelegate, then, PrompData:"Fort HUDTag Prompt Data Structure (by ref)") {is=title K2Node_Event}
   > Parent: OpenWidget!(execute, PrompData:"Fort HUDTag Prompt Data Structure (by ref)"=@n1.PrompData) -> (then) {is=title K2Node_CallParentFunction}
   > K2Node_Composite_0 = Set Text and Images!(execute, VolumePromptData:"Fort HUDTag Prompt Data Structure (by ref)"=@n1.PrompData) -> (then:Exec) {is=title K2Node_Composite}
   > UnbindAllFromAnimationFinished!(
    execute
    self:User Widget Object Reference
    Animation:Widget Animation Object Reference=Get Outro() -> (Outro:Widget Animation Object Reference)
   ) -> (then)
   > StopAnimation(InAnimation=Get Outro())
   > BindToAnimationFinished(Animation=Get Outro(), Delegate=Create Event())
   > SetVisibility(InVisibility=HitTestInvisible)
   > PlayAnimationForward(InAnimation=Get Intro())
   > RetriggerableDelay(Duration=7)
   > PlayAnimation(InAnimation=Get Outro())
  ]}
 Set Text and Images {
  path=EventGraph.K2Node_Composite_0.Set Text and Images
  nodes [
   n1 = Inputs!() -> (execute, VolumePromptData:"Fort HUDTag Prompt Data Structure (by ref)")
   n2 = Break Fort HUDTag Prompt Data!(FortHUDTagPromptData:"Fort HUDTag Prompt Data Structure (by ref)"=@n1) -> (
    VolumeTag:Gameplay Tag Structure
    Text_PromptTitle:Text
    Text_PromptDescription:Text
    bIsEnterData:bool=~
    WidgetClass:User Widget Class Reference
   ) {is=title K2Node_BreakStruct}
   > "SetText (Text)"(self=Get Text_Title() -> (Text_Title:Common Text Object Reference), InText=@n2.Text_PromptTitle)
   > "SetText (Text)"(self=Get Text_Description() -> (Text_Description:Common Text Object Reference), InText=@n2.Text_PromptDescription)
   > Outputs!(then:Exec)
  ]
 }
]
variables []
widgets {
 info {inheritedWidgetCount=1, parentClass=/Script/FortniteUI.FortHUDTagPromptWidgetBase, rootWidgetClass=&Overlay, widgetCount=14}
 widgets [
  Overlay_Container {
   widgetClassPath=&Overlay
   widget=Overlay_Container
   properties {renderTransform {angle=0, scale=<0.8 0.8>, shear=<0 0>, translation=<0 0>}, renderTransformPivot=<0.5 0>}
  }
  Image_Shadow {
   widgetClassPath=&Image
   bIsVariable=true
   parent=Overlay_Container
   widget=Image_Shadow
   properties {
    brush {
     bIsDynamicallyLoaded=false
     drawAs=Image
     imageType=NoImage
     mirroring=NoMirror
     resourceName=None
     resourceObject=/Game/Athena/HUD/LTM/Wax/Materials/M_UI_TextGlow
     tiling=NoTile
     imageSize=<775.2339 284.6466>
     margin=<ltrb 0 0 0 0>
     outlineSettings {
      bUseBrushTransparency=false
      roundingType=HalfHeightRadius
      width=0
      color {colorUseRule=UseColor_Specified, specifiedColor=<rgba 0 0 0 0>}
      cornerRadii=<0 0 0 0>
     }
     tintColor {colorUseRule=UseColor_Specified, specifiedColor=<rgba 1 1 1 1>}
     uVRegion {bIsValid=false, max=<0 0>, min=<0 0>}
    }
    colorAndOpacity=<rgba 0.0052 0.0052 0.0056 0.8>
   }
   slot {class=OverlaySlot, properties {horizontalAlignment=HAlign_Center, verticalAlignment=VAlign_Center, padding=<ltrb -800 -200 -800 -200>}}
  }
  Overlay_0 {
   widgetClassPath=&Overlay
   parent=Overlay_Container
   widget=Overlay_0
   properties {renderTransform {angle=0, scale=<1 1>, shear=<0 0>, translation=<-11 20>}}
   slot {class=OverlaySlot, properties {horizontalAlignment=HAlign_Right, verticalAlignment=VAlign_Center}}
  }
  Image_Symbol {
   widgetClassPath=&Image
   bIsVariable=true
   parent=Overlay_0
   widget=Image_Symbol
   properties {
    brush {
     bIsDynamicallyLoaded=false
     drawAs=Image
     imageType=NoImage
     mirroring=NoMirror
     resourceName=None
     resourceObject=/Game/Athena/HUD/VolumePrompt/Materials/M_UI_OakPromptSymbol_Inst
     tiling=NoTile
     imageSize=<200 250>
     margin=<ltrb 0 0 0 0>
     outlineSettings {
      bUseBrushTransparency=false
      roundingType=HalfHeightRadius
      width=0
      color {colorUseRule=UseColor_Specified, specifiedColor=<rgba 0 0 0 0>}
      cornerRadii=<0 0 0 0>
     }
     tintColor {colorUseRule=UseColor_Specified, specifiedColor=<rgba 0.3005 0.0044 0 1>}
     uVRegion {bIsValid=false, max=<0 0>, min=<0 0>}
    }
    renderTransform {angle=0, scale=<-1 1>, shear=<0 0>, translation=<0 0>}
   }
   slot {class=OverlaySlot, properties {horizontalAlignment=HAlign_Center, verticalAlignment=VAlign_Center, padding=<ltrb -150 -150 -150 -150>}}
  }
  Overlay_1 {
   widgetClassPath=&Overlay
   parent=Overlay_Container
   widget=Overlay_1
   slot {class=OverlaySlot, properties {verticalAlignment=VAlign_Center}}
  }
  SizeBox_0 {
   widgetClassPath=&SizeBox
   parent=Overlay_1
   widget=SizeBox_0
   properties {bOverride_MaxDesiredWidth=true, maxDesiredWidth=650}
   slot {class=OverlaySlot}
  }
  ScaleBox_0 {widgetClassPath=&ScaleBox, parent=SizeBox_0, widget=ScaleBox_0, slot {class=SizeBoxSlot}}
  HorizontalBox_1 {widgetClassPath=&HorizontalBox, parent=ScaleBox_0, widget=HorizontalBox_1, slot {class=ScaleBoxSlot}}
  Text_Description {
   widgetClassPath=&CommonTextBlock
   bIsVariable=true
   parent=HorizontalBox_1
   widget=Text_Description
   properties {
    justification=Center
    text=Pandora
    textTransformPolicy=ToUpper
    font {
     bForceMonospaced=false
     bMaterialIsStencil=false
     fontMaterial=/Game/Athena/HUD/VolumePrompt/Materials/M_UI_OakPrompt_Font_Inst
     fontObject=&Burbank
     letterSpacing=0
     monospacedWidth=1
     size=95
     skewAmount=0
     typefaceFontName=BurbankRegular-Black
     outlineSettings {
      bApplyOutlineToDropShadows=true
      bMiteredCorners=false
      bSeparateFillAlpha=false
      outlineMaterial=null
      outlineSize=4
      outlineColor=<rgba 0.0052 0.0052 0.0056 1>
     }
    }
    renderTransform {angle=-1.5, scale=<1 1>, shear=<5 0>, translation=<0 0>}
    shadowColorAndOpacity=<rgba 0.0052 0.0052 0.0056 1>
    shadowOffset=<-2.2361 12>
   }
   slot {class=HorizontalBoxSlot}
  }
  SizeBox_1 {
   widgetClassPath=&SizeBox
   parent=Overlay_1
   widget=SizeBox_1
   properties {bOverride_MaxDesiredWidth=true, maxDesiredWidth=650, renderTransform {angle=-4, scale=<1 1>, shear=<0 0>, translation=<-61 9>}}
   slot {class=OverlaySlot}
  }
  ScaleBox_1 {widgetClassPath=&ScaleBox, parent=SizeBox_1, widget=ScaleBox_1, properties {stretchDirection=DownOnly}, slot {class=SizeBoxSlot}}
  Text_Title {
   widgetClassPath=&CommonTextBlock
   bIsVariable=true
   parent=ScaleBox_1
   widget=Text_Title
   properties {
    justification=Right
    text=Welcome to
    textTransformPolicy=ToUpper
    font {
     bForceMonospaced=false
     bMaterialIsStencil=false
     fontMaterial=null
     fontObject=&BurbankSmall
     letterSpacing=0
     monospacedWidth=1
     size=24
     skewAmount=0
     typefaceFontName=Bold
     outlineSettings {
      bApplyOutlineToDropShadows=false
      bMiteredCorners=false
      bSeparateFillAlpha=false
      outlineMaterial=null
      outlineSize=3
      outlineColor=<rgba 0.0052 0.0052 0.0056 1>
     }
    }
    renderTransform {angle=0, scale=<1 1>, shear=<-9 0>, translation=<0 0>}
    renderTransformPivot=<1.8927 2.7513>
    shadowColorAndOpacity=<rgba 0.0052 0.0052 0.0056 1>
    shadowOffset=<0 0>
   }
   slot {class=ScaleBoxSlot}
  }
  Text_Title_1 {
   widgetClassPath=&CommonTextBlock
   bIsVariable=true
   parent=Overlay_1
   widget=Text_Title_1
   properties {
    justification=Right
    text="Shields Up, Traveler!"
    textTransformPolicy=ToUpper
    font {
     bForceMonospaced=false
     bMaterialIsStencil=false
     fontMaterial=null
     fontObject=&BurbankSmall
     letterSpacing=0
     monospacedWidth=1
     size=18
     skewAmount=0
     typefaceFontName=Bold
     outlineSettings {
      bApplyOutlineToDropShadows=false
      bMiteredCorners=false
      bSeparateFillAlpha=false
      outlineMaterial=null
      outlineSize=3
      outlineColor=<rgba 0.0052 0.0052 0.0056 1>
     }
    }
    renderTransform {angle=1, scale=<1 1>, shear=<-9 -3>, translation=<21 -5>}
    renderTransformPivot=<0.5 1>
    shadowColorAndOpacity=<rgba 0.0052 0.0052 0.0056 1>
    shadowOffset=<0 0>
   }
   slot {class=OverlaySlot, properties {horizontalAlignment=HAlign_Right, verticalAlignment=VAlign_Bottom}}
  }
  TutorialHighlight {widgetClassPath=&TutorialHighlight}
 ]
}
