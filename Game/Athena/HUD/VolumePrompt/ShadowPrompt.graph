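# HUD tag prompt widget that announces a Shadow or Ghost disguise with a glitch symbol and two-line title.
# PreConstruct picks icon and text from isShadow unless CustomLogo or CustomText overrides them, splitting text at the line break.
# OpenWidget plays Intro then Outro, ignoring the prompt data; CloseWidget collapses it.
# Two counter-rotated clipping borders slice the faction name diagonally, and Intro slides the halves together.
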WidgetBlueprint /Game/Athena/HUD/VolumePrompt/ShadowPrompt
animations [
 Intro {
  binding_count=6
  length_seconds=1.75
  object=&self:Intro
  sequence {bindings [
    CommonTextBlock_Faction2 {
     id=#1
     possessed_class=CommonTextBlock
     tracks [
      {
       type=MovieScene2DTransformTrack
       sections [{
         type=MovieScene2DTransformSection
         channels [
          Translation.X {type=MovieSceneScriptingFloatChannel, keys [[12.1843,334],[20.2167,0]]}
          Translation.Y {type=MovieSceneScriptingFloatChannel}
          Angle {type=MovieSceneScriptingFloatChannel}
          Scale.X {type=MovieSceneScriptingFloatChannel}
          Scale.Y {type=MovieSceneScriptingFloatChannel}
          Shear.X {type=MovieSceneScriptingFloatChannel}
          Shear.Y {type=MovieSceneScriptingFloatChannel}
         ]
        }]
      }
      {
       type=MovieSceneFloatTrack
       display_name=Render Opacity
       sections [{type=MovieSceneFloatSection, channels [None {type=MovieSceneScriptingFloatChannel, keys [[12.1843,0],[14.691,1],[20.2167,0]]}]}]
      }
     ]
    }
    CommonTextBlock_Faction3 {
     id=#2
     possessed_class=CommonTextBlock
     tracks [
      {
       type=MovieScene2DTransformTrack
       sections [{
         type=MovieScene2DTransformSection
         channels [
          Translation.X {type=MovieSceneScriptingFloatChannel, keys [[12.1843,-334],[20.2167,0]]}
          Translation.Y {type=MovieSceneScriptingFloatChannel}
          Angle {type=MovieSceneScriptingFloatChannel}
          Scale.X {type=MovieSceneScriptingFloatChannel}
          Scale.Y {type=MovieSceneScriptingFloatChannel}
          Shear.X {type=MovieSceneScriptingFloatChannel}
          Shear.Y {type=MovieSceneScriptingFloatChannel}
         ]
        }]
      }
      {
       type=MovieSceneFloatTrack
       display_name=Render Opacity
       sections [{type=MovieSceneFloatSection, channels [None {type=MovieSceneScriptingFloatChannel, keys [[12.1843,0],[14.691,1],[20.2167,0]]}]}]
      }
     ]
    }
    CommonTextBlock_Disguise {
     id=#3
     possessed_class=CommonTextBlock
     tracks [
      {
       type=MovieSceneFloatTrack
       display_name=Render Opacity
       sections [{type=MovieSceneFloatSection, channels [None {type=MovieSceneScriptingFloatChannel, keys [[20.2167,0],[25.025,1]]}]}]
      }
      {
       type=MovieScene2DTransformTrack
       sections [{
         type=MovieScene2DTransformSection
         channels [
          Translation.X {type=MovieSceneScriptingFloatChannel}
          Translation.Y {type=MovieSceneScriptingFloatChannel, keys [[20.2167,-37.9162],[25.025,0]]}
          Angle {type=MovieSceneScriptingFloatChannel}
          Scale.X {type=MovieSceneScriptingFloatChannel}
          Scale.Y {type=MovieSceneScriptingFloatChannel}
          Shear.X {type=MovieSceneScriptingFloatChannel}
          Shear.Y {type=MovieSceneScriptingFloatChannel}
         ]
        }]
      }
     ]
    }
    CommonBorderCleanText {
     id=#4
     possessed_class=CommonBorder
     tracks [{
       type=MovieSceneFloatTrack
       display_name=Render Opacity
       sections [{type=MovieSceneFloatSection, channels [None {type=MovieSceneScriptingFloatChannel, keys [[12.1843,0],[20.2167,1]]}]}]
      }]
    }
    Overlay_Container {
     id=#5
     possessed_class=Overlay
     tracks [
      {
       type=MovieSceneFloatTrack
       display_name=Render Opacity
       sections [{type=MovieSceneFloatSection, channels [None {type=MovieSceneScriptingFloatChannel, keys [[0,0],[3.691,1],[35,1]]}]}]
      }
      {
       type=MovieScene2DTransformTrack
       sections [{
         type=MovieScene2DTransformSection
         channels [
          Translation.X {type=MovieSceneScriptingFloatChannel}
          Translation.Y {type=MovieSceneScriptingFloatChannel}
          Angle {type=MovieSceneScriptingFloatChannel}
          Scale.X {type=MovieSceneScriptingFloatChannel, keys [[26.6327,1],[31.7787,0.75]]}
          Scale.Y {type=MovieSceneScriptingFloatChannel, keys [[26.6327,1],[31.7787,0.75]]}
          Shear.X {type=MovieSceneScriptingFloatChannel}
          Shear.Y {type=MovieSceneScriptingFloatChannel}
         ]
        }]
      }
     ]
    }
    Image_Symbol {
     id=#6
     possessed_class=Image
     tracks [
      {
       type=MovieSceneWidgetMaterialTrack
       sections [{
         type=MovieSceneParameterSection
         channels [ForceAlphaChange {type=MovieSceneScriptingFloatChannel, keys [[3.691,13.8428],[14.691,1.2866],[17.598,0.675]]}]
        }]
      }
      {
       type=MovieScene2DTransformTrack
       sections [{
         type=MovieScene2DTransformSection
         channels [
          Translation.X {type=MovieSceneScriptingFloatChannel, keys [[10.691,180.1676],[14.382,0]]}
          Translation.Y {type=MovieSceneScriptingFloatChannel}
          Angle {type=MovieSceneScriptingFloatChannel}
          Scale.X {type=MovieSceneScriptingFloatChannel, keys [[3.691,1.2],[10.691,1]]}
          Scale.Y {type=MovieSceneScriptingFloatChannel, keys [[3.691,1.2],[10.691,1]]}
          Shear.X {type=MovieSceneScriptingFloatChannel}
          Shear.Y {type=MovieSceneScriptingFloatChannel}
         ]
        }]
      }
     ]
    }
   ]}
 }
 Outro {
  binding_count=1
  length_seconds=0.7391
  object=&self:Outro
  sequence {bindings [Overlay_Container {
     id=#7
     possessed_class=Overlay
     tracks [{
       type=MovieSceneFloatTrack
       display_name=Render Opacity
       sections [{type=MovieSceneFloatSection, end_frame=14, start_frame=0, channels [None {type=MovieSceneScriptingFloatChannel, keys [[0,1],[14.7813,0]]}]}]
      }]
    }]}
 }
]
graphs [
 EventGraph {nodes [
   n1 = IntroFinished() {is=title K2Node_CustomEvent}
   > PlayAnimation(InAnimation=Get Outro() -> (Outro:Widget Animation Object Reference))
   Event CloseWidget() {is=title K2Node_Event}
   > Parent: CloseWidget() {is=title K2Node_CallParentFunction}
   > SetVisibility(InVisibility=Collapsed)
   Event PreConstruct()
   n2 = Get CustomLogo() -> (CustomLogo:Material Interface Object Reference)
   n3 = Get isShadow() -> (isShadow:bool)
   > SetBrushFromMaterial(
    self=Get Image_Symbol() -> (Image_Symbol:Image Object Reference)
    Material=Select(
     Option 0:Material Interface Object Reference=Select(
      Option 0=/Game/Athena/UI/Frontend/Spatial/Art/InteractionIndicators/M_UI_SelectionGlitch_Icon_White_Ego_Prompt
      Option 1=/Game/Athena/UI/Frontend/Spatial/Art/InteractionIndicators/M_UI_SelectionGlitch_Icon_White_Alter_Prompt
      Index=@n3
     )
     Option 1:Material Interface Object Reference=@n2
     Index=IsValid(Object=@n2)
    ) -> (ReturnValue:Material Interface Object Reference)
   )
   n4 = Get CustomText() -> (CustomText:Text)
   n5 = TextIsEmpty(InText=@n4)
   n6 = Select(Option 0:Text="LOCTEXT(\"Ghost\\r\\nDisguise\")", Option 1:Text="LOCTEXT(\"Shadow\\r\\nDisguise\")", Index=@n3) -> (ReturnValue:Text)
   n7 = Split(
    SourceString="To String (Text)"(InText=Select(Option 0:"Text (by ref)"=@n4, Option 1:"Text (by ref)"=@n6, Index=@n5) -> (ReturnValue:"Text (by ref)"))
    InStr="\r\n"
    SearchDir=FromStart
   )
   > "SetText (Text)"(
    self=[
     Get CommonTextBlock_Faction1() -> (CommonTextBlock_Faction1:Common Text Object Reference)
     Get CommonTextBlock_Faction2() -> (CommonTextBlock_Faction2:Common Text Object Reference)
     Get CommonTextBlock_Faction3() -> (CommonTextBlock_Faction3:Common Text Object Reference)
    ]
    InText="To Text (String)"(InString=@n7)
   )
   > "SetText (Text)"(
    self=Get CommonTextBlock_Disguise() -> (CommonTextBlock_Disguise:Common Text Object Reference)
    InText="To Text (String)"(InString=@n7.RightS)
   )
   n8 = Event OpenWidget!() -> (OutputDelegate, then, PrompData:"Fort HUDTag Prompt Data Structure (by ref)") {is=title K2Node_Event}
   > Parent: OpenWidget!(execute, PrompData:"Fort HUDTag Prompt Data Structure (by ref)"=@n8.PrompData) -> (then) {is=title K2Node_CallParentFunction}
   > K2Node_Composite_0 = Set Text and Images!(execute, VolumePromptData:"Fort HUDTag Prompt Data Structure (by ref)"=@n8.PrompData) -> (then:Exec) {is=title K2Node_Composite}
   > UnbindAllFromAnimationFinished!(execute, self:User Widget Object Reference, Animation:Widget Animation Object Reference=Get Outro()) -> (then)
   > StopAnimation(InAnimation=Get Outro())
   > BindToAnimationFinished(Animation=Get Outro(), Delegate=Create Event())
   > SetVisibility(InVisibility=HitTestInvisible)
   n9 = Get Intro()
   > PlayAnimationForward(InAnimation=@n9)
   > BindToAnimationFinished(Animation=@n9, Delegate=@n1)
  ]}
 Set Text and Images {
  path=EventGraph.K2Node_Composite_0.Set Text and Images
  nodes [
   n1 = Inputs!() -> (execute, VolumePromptData:"Fort HUDTag Prompt Data Structure (by ref)")
   > Outputs!(then:Exec)
   Break Fort HUDTag Prompt Data!(FortHUDTagPromptData:"Fort HUDTag Prompt Data Structure (by ref)"=@n1) -> (
    VolumeTag:Gameplay Tag Structure
    Text_PromptTitle:Text
    Text_PromptDescription:Text
    bIsEnterData:bool=~
    WidgetClass:User Widget Class Reference
   ) {is=title K2Node_BreakStruct}
  ]
 }
]
variables [
 CustomLogo {type=MaterialInterface, default=null}
 CustomText {type=string, default=""}
 isShadow {type=boolean, default=true}
]
widgets {
 info {inheritedWidgetCount=1, parentClass=/Script/FortniteUI.FortHUDTagPromptWidgetBase, rootWidgetClass=&Overlay, widgetCount=27}
 widgets [
  Overlay_Container {
   widgetClassPath=&Overlay
   widget=Overlay_Container
   properties {renderTransform {angle=0, scale=<0.8 0.8>, shear=<0 0>, translation=<0 0>}, renderTransformPivot=<0.5 0>}
  }
  Image_Shadow {
   widgetClassPath=&Image
   bIsVariable=true
   parent=Overlay_Container
   widget=Image_Shadow
   properties {
    brush {
     bIsDynamicallyLoaded=false
     drawAs=Image
     imageType=NoImage
     mirroring=NoMirror
     resourceName=None
     resourceObject=/Game/Athena/HUD/LTM/Wax/Materials/M_UI_TextGlow
     tiling=NoTile
     imageSize=<775.2339 284.6466>
     margin=<ltrb 0 0 0 0>
     outlineSettings {
      bUseBrushTransparency=false
      roundingType=HalfHeightRadius
      width=0
      color {colorUseRule=UseColor_Specified, specifiedColor=<rgba 0 0 0 0>}
      cornerRadii=<0 0 0 0>
     }
     tintColor {colorUseRule=UseColor_Specified, specifiedColor=<rgba 1 1 1 1>}
     uVRegion {bIsValid=false, max=<0 0>, min=<0 0>}
    }
    colorAndOpacity=<rgba 0.0052 0.0052 0.0056 0.6>
   }
   slot {class=OverlaySlot, properties {horizontalAlignment=HAlign_Center, verticalAlignment=VAlign_Center, padding=<ltrb -800 -200 -800 -200>}}
  }
  Overlay_0 {
   widgetClassPath=&Overlay
   parent=Overlay_Container
   widget=Overlay_0
   properties {renderTransform {angle=0, scale=<1 1>, shear=<0 0>, translation=<14 15>}}
   slot {class=OverlaySlot, properties {verticalAlignment=VAlign_Center}}
  }
  Image_Symbol {
   widgetClassPath=&Image
   bIsVariable=true
   parent=Overlay_0
   widget=Image_Symbol
   properties {
    brush {
     bIsDynamicallyLoaded=false
     drawAs=Image
     imageType=NoImage
     mirroring=NoMirror
     resourceName=None
     resourceObject=&M_UI_SelectionGlitch_Icon_White_Alter_Prompt
     tiling=NoTile
     imageSize=<300 300>
     margin=<ltrb 0 0 0 0>
     outlineSettings {
      bUseBrushTransparency=false
      roundingType=HalfHeightRadius
      width=0
      color {colorUseRule=UseColor_Specified, specifiedColor=<rgba 0 0 0 0>}
      cornerRadii=<0 0 0 0>
     }
     tintColor {colorUseRule=UseColor_Specified, specifiedColor=<rgba 1 1 1 1>}
     uVRegion {bIsValid=false, max=<0 0>, min=<0 0>}
    }
    renderTransform {angle=-10, scale=<1 1>, shear=<0 0>, translation=<0 0>}
   }
   slot {class=OverlaySlot, properties {horizontalAlignment=HAlign_Center, verticalAlignment=VAlign_Center, padding=<ltrb -150 -150 -150 -150>}}
  }
  ScaleBox_2 {
   widgetClassPath=&ScaleBox
   parent=Overlay_Container
   widget=ScaleBox_2
   properties {stretch=UserSpecified}
   slot {class=OverlaySlot, properties {verticalAlignment=VAlign_Center}}
  }
  VerticalBox_0 {
   widgetClassPath=&VerticalBox
   parent=ScaleBox_2
   widget=VerticalBox_0
   properties {renderTransform {angle=0, scale=<1 1>, shear=<0 0>, translation=<0 13>}}
   slot {class=ScaleBoxSlot}
  }
  Overlay_2 {
   widgetClassPath=&Overlay
   parent=VerticalBox_0
   widget=Overlay_2
   properties {renderTransform {angle=0, scale=<1 1>, shear=<0 0>, translation=<-31 0>}}
   slot {class=VerticalBoxSlot}
  }
  CommonBorderCleanText {
   widgetClassPath=&CommonBorder
   parent=Overlay_2
   widget=CommonBorderCleanText
   properties {padding=<ltrb 30 0 30 0>}
   slot {class=OverlaySlot}
  }
  Overlay_3 {
   widgetClassPath=&Overlay
   parent=CommonBorderCleanText
   widget=Overlay_3
   slot {class=BorderSlot, properties {padding=<ltrb 30 0 30 0>}}
  }
  SizeBox_1 {
   widgetClassPath=&SizeBox
   parent=Overlay_3
   widget=SizeBox_1
   properties {bOverride_MaxDesiredWidth=true, maxDesiredWidth=800}
   slot {class=OverlaySlot, properties {verticalAlignment=VAlign_Center}}
  }
  ScaleBox_1 {
   widgetClassPath=&ScaleBox
   parent=SizeBox_1
   widget=ScaleBox_1
   slot {class=SizeBoxSlot, properties {verticalAlignment=VAlign_Center}}
  }
  HorizontalBox_2 {widgetClassPath=&HorizontalBox, parent=ScaleBox_1, widget=HorizontalBox_2, slot {class=ScaleBoxSlot}}
  CommonTextBlock_Faction1 {
   widgetClassPath=&CommonTextBlock
   bIsVariable=true
   parent=HorizontalBox_2
   widget=CommonTextBlock_Faction1
   properties {
    text=Shadow
    textTransformPolicy=ToUpper
    font {
     bForceMonospaced=false
     bMaterialIsStencil=false
     fontMaterial=null
     fontObject=&Burbank
     letterSpacing=0
     monospacedWidth=1
     size=60
     skewAmount=0
     typefaceFontName=BurbankRegular-Black
     outlineSettings {
      bApplyOutlineToDropShadows=false
      bMiteredCorners=false
      bSeparateFillAlpha=false
      outlineMaterial=/Game/UI/Foundation/Materials/M-FontShadow
      outlineSize=5
      outlineColor=<rgba 0 0 0 1>
     }
    }
    renderTransform {angle=0, scale=<1 1>, shear=<-12 0>, translation=<30 -1>}
    shadowColorAndOpacity=<rgba 0 0 0 1>
    shadowOffset=<0 8.06>
   }
   slot {class=HorizontalBoxSlot, properties {horizontalAlignment=HAlign_Left, verticalAlignment=VAlign_Bottom, padding=<ltrb 30 20 30 20>}}
  }
  BorderClippedText_Left {
   widgetClassPath=&CommonBorder
   parent=Overlay_2
   widget=BorderClippedText_Left
   properties {
    clipping=ClipToBounds
    horizontalAlignment=HAlign_Center
    verticalAlignment=VAlign_Center
    padding=<ltrb 30 0 30 0>
    renderTransform {angle=-17, scale=<1 1>, shear=<0 0>, translation=<2 -62>}
   }
   slot {class=OverlaySlot}
  }
  Overlay_1 {
   widgetClassPath=&Overlay
   parent=BorderClippedText_Left
   widget=Overlay_1
   properties {renderTransform {angle=17, scale=<1 1>, shear=<0 0>, translation=<9 67>}}
   slot {class=BorderSlot, properties {horizontalAlignment=HAlign_Center, verticalAlignment=VAlign_Center, padding=<ltrb 30 0 30 0>}}
  }
  SizeBox_0 {
   widgetClassPath=&SizeBox
   parent=Overlay_1
   widget=SizeBox_0
   properties {bOverride_MaxDesiredWidth=true, maxDesiredWidth=800}
   slot {class=OverlaySlot, properties {verticalAlignment=VAlign_Center}}
  }
  ScaleBox_0 {
   widgetClassPath=&ScaleBox
   parent=SizeBox_0
   widget=ScaleBox_0
   slot {class=SizeBoxSlot, properties {verticalAlignment=VAlign_Center}}
  }
  HorizontalBox_1 {widgetClassPath=&HorizontalBox, parent=ScaleBox_0, widget=HorizontalBox_1, slot {class=ScaleBoxSlot}}
  CommonTextBlock_Faction2 {
   widgetClassPath=&CommonTextBlock
   bIsVariable=true
   parent=HorizontalBox_1
   widget=CommonTextBlock_Faction2
   properties {
    text=Shadow
    textTransformPolicy=ToUpper
    font {
     bForceMonospaced=false
     bMaterialIsStencil=false
     fontMaterial=null
     fontObject=&Burbank
     letterSpacing=0
     monospacedWidth=1
     size=60
     skewAmount=0
     typefaceFontName=BurbankRegular-Black
     outlineSettings {
      bApplyOutlineToDropShadows=false
      bMiteredCorners=false
      bSeparateFillAlpha=false
      outlineMaterial=&M-FontShadow
      outlineSize=5
      outlineColor=<rgba 0 0 0 1>
     }
    }
    renderTransform {angle=0, scale=<1 1>, shear=<-12 0>, translation=<0 0>}
    shadowColorAndOpacity=<rgba 0 0 0 1>
    shadowOffset=<0 8.06>
   }
   slot {class=HorizontalBoxSlot, properties {horizontalAlignment=HAlign_Left, verticalAlignment=VAlign_Bottom, padding=<ltrb 30 20 30 20>}}
  }
  BorderClippedText_Right {
   widgetClassPath=&CommonBorder
   parent=Overlay_2
   widget=BorderClippedText_Right
   properties {
    clipping=ClipToBounds
    horizontalAlignment=HAlign_Center
    verticalAlignment=VAlign_Center
    padding=<ltrb 30 0 30 0>
    renderTransform {angle=-17, scale=<1 1>, shear=<0 0>, translation=<45 73>}
   }
   slot {class=OverlaySlot}
  }
  Overlay {
   widgetClassPath=&Overlay
   parent=BorderClippedText_Right
   widget=Overlay
   properties {renderTransform {angle=17, scale=<1 1>, shear=<0 0>, translation=<7 -75>}}
   slot {class=BorderSlot, properties {horizontalAlignment=HAlign_Center, verticalAlignment=VAlign_Center, padding=<ltrb 30 0 30 0>}}
  }
  SizeBox {
   widgetClassPath=&SizeBox
   parent=Overlay
   widget=SizeBox
   properties {bOverride_MaxDesiredWidth=true, maxDesiredWidth=800}
   slot {class=OverlaySlot, properties {verticalAlignment=VAlign_Center}}
  }
  ScaleBox {
   widgetClassPath=&ScaleBox
   parent=SizeBox
   widget=ScaleBox
   slot {class=SizeBoxSlot, properties {verticalAlignment=VAlign_Center}}
  }
  HorizontalBox {widgetClassPath=&HorizontalBox, parent=ScaleBox, widget=HorizontalBox, slot {class=ScaleBoxSlot}}
  CommonTextBlock_Faction3 {
   widgetClassPath=&CommonTextBlock
   bIsVariable=true
   parent=HorizontalBox
   widget=CommonTextBlock_Faction3
   properties {
    text=Shadow
    textTransformPolicy=ToUpper
    font {
     bForceMonospaced=false
     bMaterialIsStencil=false
     fontMaterial=null
     fontObject=&Burbank
     letterSpacing=0
     monospacedWidth=1
     size=60
     skewAmount=0
     typefaceFontName=BurbankRegular-Black
     outlineSettings {
      bApplyOutlineToDropShadows=false
      bMiteredCorners=false
      bSeparateFillAlpha=false
      outlineMaterial=&M-FontShadow
      outlineSize=5
      outlineColor=<rgba 0 0 0 1>
     }
    }
    renderTransform {angle=0, scale=<1 1>, shear=<-12 0>, translation=<0 0>}
    shadowColorAndOpacity=<rgba 0 0 0 1>
    shadowOffset=<0 8.06>
   }
   slot {class=HorizontalBoxSlot, properties {horizontalAlignment=HAlign_Left, verticalAlignment=VAlign_Bottom, padding=<ltrb 30 20 30 20>}}
  }
  CommonTextBlock_Disguise {
   widgetClassPath=&CommonTextBlock
   bIsVariable=true
   parent=VerticalBox_0
   widget=CommonTextBlock_Disguise
   properties {
    text=Disguise
    textTransformPolicy=ToUpper
    font {
     bForceMonospaced=false
     bMaterialIsStencil=false
     fontMaterial=null
     fontObject=&Burbank
     letterSpacing=0
     monospacedWidth=1
     size=24
     skewAmount=0
     typefaceFontName=BurbankRegular-Black
     outlineSettings {
      bApplyOutlineToDropShadows=false
      bMiteredCorners=false
      bSeparateFillAlpha=false
      outlineMaterial=&M-FontShadow
      outlineSize=1
      outlineColor=<rgba 0 0 0 1>
     }
    }
    renderTransform {angle=0, scale=<1 1>, shear=<-8 0>, translation=<0 0>}
    shadowColorAndOpacity=<rgba 0 0 0 1>
    shadowOffset=<0 5.21>
   }
   slot {class=VerticalBoxSlot, properties {horizontalAlignment=HAlign_Left, verticalAlignment=VAlign_Center, padding=<ltrb 60 -35 0 0>}}
  }
  TutorialHighlight {widgetClassPath=&TutorialHighlight}
 ]
}
